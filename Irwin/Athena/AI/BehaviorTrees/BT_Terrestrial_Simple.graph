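# Simple behavior tree for a terrestrial NPC, choosing non-combat behaviors through a priority selector and tag-identified gameplay abilities.
# It idles when ridden (setting InRiddenBTNode), tractor-beamed or suspended.
# Otherwise it avoids goal pawns within ability range or during an active avoid phase, with a service sending the phase-apply event.
# Lower branches reposition around a Leader, move to land when swimming, or wander after a random delay, with a fallback wait.
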BehaviorTree /Irwin/Athena/AI/BehaviorTrees/BT_Terrestrial_Simple
blackboard=/Game/Athena/AI/Blackboards/FortAIBB_NPC_Default
root {
 name=Selector
 class=BTComposite_Selector
 children [
  Sequence {
   class=BTComposite_Sequence
   children [
    SetRidingKey {
     class=Task_SetBlackboardBoolValues_C
     children []
     decorators []
     properties {
      bIgnoreRestartSelf=false
      bShowPropertyDetails=true
      boolKey1Value=true
      boolKey2Value=false
      boolKey3Value=false
      boolKey4Value=false
      boolKey5Value=false
      customDescription=""
      nodeName=SetRidingKey
      update 1?=true
      update 2?=false
      update 3?=false
      update 4?=false
      update 5?=false
      boolKey1 {bNoneIsAllowedValue=false, selectedKeyName=InRiddenBTNode}
      boolKey2 {bNoneIsAllowedValue=false, selectedKeyName=GoalActor}
      boolKey3 {bNoneIsAllowedValue=false, selectedKeyName=GoalActor}
      boolKey4 {bNoneIsAllowedValue=false, selectedKeyName=GoalActor}
      boolKey5 {bNoneIsAllowedValue=false, selectedKeyName=GoalActor}
      tickInterval {interval=-1}
     }
     services []
    }
    Wait {class=BTTask_Wait, properties {waitTime {defaultValue=2}}}
   ]
   decorators [IsBeingRidden {
     class=BTDecorator_GameplayTagQuery
     properties {
      flowAbortMode=Both
      nodeName=IsBeingRidden
      gameplayTagQuery {autoDescription=" ANY( Riding.IsBeingRidden )", queryTokenStream [0,1,1,1,0], tagDictionary [{tagName=Riding.IsBeingRidden}]}
     }
    }]
   properties {children [
     {childTask=&self:Task_SetBlackboardBoolValues_C_1}
     {childTask=&self:BTTask_Wait_18}
    ]}
  }
  Wait {
   class=BTTask_Wait
   decorators [Status TractorBeam or SuspendedInAir {
     class=BTDecorator_GameplayTagQuery
     properties {
      flowAbortMode=None
      nodeName=Status TractorBeam or SuspendedInAir
      gameplayTagQuery {
       autoDescription=" ANY( Gameplay.Status.TractorBeam, Gameplay.Status.SuspendedInAir )"
       queryTokenStream [0,1,1,2,0,1]
       tagDictionary [{tagName=Gameplay.Status.TractorBeam}, {tagName=Gameplay.Status.SuspendedInAir}]
      }
     }
    }]
   properties {waitTime {defaultValue=2}}
  }
  Selector {
   class=BTComposite_Selector
   children [GameplayAbility {
     class=FortBTTask_ExecuteGameplayAbility
     decorators [
      Has Gameplay Ability {
       class=FortBTDecorator_GameplayAbility_HasGameplayAbility
       properties {gameplayAbilityTagContainer {defaultValue {gameplayTags [{tagName=NPC.Ability.NonAttack.Avoid.Enemy.Primary}]}}}
      }
      Can Activate Gameplay Ability {
       class=FortBTDecorator_GameplayAbility_CanActivate
       properties {
        flowAbortMode=None
        gameplayAbilityTagContainer {defaultValue {gameplayTags [{tagName=NPC.Ability.NonAttack.Avoid.Enemy.Primary}]}}
        targetValue {key=GoalActor}
       }
      }
     ]
     properties {nodeName=GameplayAbility, gameplayAbilityTagContainer {defaultValue {gameplayTags [{tagName=NPC.Ability.NonAttack.Avoid.Enemy.Primary}]}}}
    }]
   decorators [
    Has Goal Actor {
     class=BTDecorator_Blackboard
     properties {
      arithmeticOperation=Equal
      flowAbortMode=None
      intValue=0
      nodeName=Has Goal Actor
      notifyObserver=ValueChange
      stringValue=""
      blackboardKey {selectedKeyName=GoalActor}
     }
    }
    Is Goal Pawn {
     class=FortBTDecorator_IsGoalPawn
     properties {bInverseCondition=false, flowAbortMode=None, nodeName=Is Goal Pawn, blackboardKey {bNoneIsAllowedValue=false, selectedKeyName=GoalActor}}
    }
    Compare Gameplay Ability Distances {
     class=FortBTDecorator_GameplayAbility_CompareDistance
     properties {
      bInverseCondition=false
      flowAbortMode=LowerPriority
      nodeName=Compare Gameplay Ability Distances
      activeAbilityTagsToSkipTestingValue {key=None, defaultValue {gameplayTags []}}
      distanceComparisons [{
        addedOffset=0
        bUseAddedOffset=false
        bUseOverriddenValue=false
        comparisonType=ThreeDimensions
        operator=Less
        overriddenValue=0
        distanceDataTags {gameplayTags [{tagName=NPC.Ability.MaxApproachRange}]}
       }]
      gameplayAbilityTagContainer {key=None, defaultValue {gameplayTags [{tagName=NPC.Ability.NonAttack.Avoid.Enemy.Primary}]}}
      targetValue {baseClass=&Actor, defaultValue=null, key=GoalActor}
     }
    }
    Gameplay Tag Query {
     class=BTDecorator_GameplayTagQuery
     properties {
      flowAbortMode=LowerPriority
      nodeName=Gameplay Tag Query
      gameplayTagQuery {
       autoDescription=" ANY( NPC.Behavior.Phase.Avoid.Enemy.Active )"
       queryTokenStream [0,1,1,1,0]
       tagDictionary [{tagName=NPC.Behavior.Phase.Avoid.Enemy.Active}]
      }
     }
    }
   ]
   properties {
    children [{
      childComposite=null
      childTask=&self:FortBTTask_ExecuteGameplayAbility_2
      decorators [
       &self:FortBTDecorator_GameplayAbility_HasGameplayAbility_2
       &self:FortBTDecorator_GameplayAbility_CanActivate_2
      ]
     }]
    services [&self:FortBTService_SendGameplayEvents_C_3]
   }
   services [Send Gameplay Events to Blackboard Key - Avoid Enemy Phase {
     class=FortBTService_SendGameplayEvents_C
     properties {
      activatingPawn=null
      activatingPawnAbilitySystemComponent=null
      bCallTickOnSearchStart=false
      bRestartTimerOnEachActivation=false
      bShowEventDetails=false
      bShowPropertyDetails=true
      customDescription=""
      goalActor=null
      interval=1
      nodeName=Send Gameplay Events to Blackboard Key - Avoid Enemy Phase
      ownerActor=null
      randomDeviation=0.1
      targetAbilitySystemComponent=null
      useEventTickLogic=false
      activationEventKeysTags []
      activationEventValuesBlackboards []
      activationEvents {"(TagName=\"NPC.Behavior.Phase.Avoid.Enemy.Apply\")" {bNoneIsAllowedValue=false, selectedKeyName=SelfActor}}
      deactivationEventKeysTags []
      deactivationEvents {}
      deactivationEventsValuesBlackboards []
     }
    }]
  }
  Selector {
   class=BTComposite_Selector
   children [
    GameplayAbility {
     class=FortBTTask_ExecuteGameplayAbility
     decorators [
      Has Gameplay Ability {
       class=FortBTDecorator_GameplayAbility_HasGameplayAbility
       properties {gameplayAbilityTagContainer {defaultValue {gameplayTags [{tagName=NPC.Ability.NonAttack.Reposition.AroundLeader}]}}}
      }
      Can Activate Gameplay Ability {
       class=FortBTDecorator_GameplayAbility_CanActivate
       properties {
        flowAbortMode=None
        gameplayAbilityTagContainer {defaultValue {gameplayTags [{tagName=NPC.Ability.NonAttack.Reposition.AroundLeader}]}}
        targetValue {key=SpawnSpot}
       }
      }
     ]
     properties {
      nodeName=GameplayAbility
      gameplayAbilityTagContainer {defaultValue {gameplayTags [{tagName=NPC.Ability.NonAttack.Reposition.AroundLeader}]}}
     }
    }
    Wait {
     class=BTTask_Wait
     decorators [Is On Cooldown {
       class=FortBTDecorator_GameplayAbility_IsOnCooldown
       properties {
        bInverseCondition=false
        flowAbortMode=None
        nodeName=Is On Cooldown
        activeAbilityTagsToSkipTestingValue {key=None, defaultValue {gameplayTags []}}
        gameplayAbilityTagContainer {key=None, defaultValue {gameplayTags [{tagName=NPC.Ability.NonAttack.Reposition.AroundLeader}]}}
        targetValue {baseClass=&Actor, defaultValue=null, key=SpawnSpot}
       }
      }]
     properties {waitTime {defaultValue=1}}
    }
   ]
   decorators [Blackboard Based Condition {
     class=BTDecorator_Blackboard
     properties {
      arithmeticOperation=Equal
      intValue=0
      nodeName=Blackboard Based Condition
      notifyObserver=ValueChange
      stringValue=""
      blackboardKey {selectedKeyName=Leader}
     }
    }]
   properties {
    children [
     {
      childComposite=null
      childTask=&self:FortBTTask_ExecuteGameplayAbility_6
      decorators [
       &self:FortBTDecorator_GameplayAbility_HasGameplayAbility_0
       &self:FortBTDecorator_GameplayAbility_CanActivate_10
      ]
     }
     {
      childComposite=null
      childTask=&self:BTTask_Wait_3
      decorators [&self:FortBTDecorator_GameplayAbility_IsOnCooldown_3]
     }
    ]
    services []
   }
   services []
  }
  Selector {
   class=BTComposite_Selector
   children [
    GameplayAbility {
     class=FortBTTask_ExecuteGameplayAbility
     decorators [Can Activate Gameplay Ability {
       class=FortBTDecorator_GameplayAbility_CanActivate
       properties {
        flowAbortMode=None
        gameplayAbilityTagContainer {defaultValue {gameplayTags [{tagName=NPC.Ability.NonAttack.Move.ToLand}]}}
        targetValue {key=SpawnSpot}
       }
      }]
     properties {nodeName=GameplayAbility, gameplayAbilityTagContainer {defaultValue {gameplayTags [{tagName=NPC.Ability.NonAttack.Move.ToLand}]}}}
    }
    Wait {
     class=BTTask_Wait
     decorators [Is On Cooldown {
       class=FortBTDecorator_GameplayAbility_IsOnCooldown
       properties {
        bInverseCondition=false
        flowAbortMode=None
        nodeName=Is On Cooldown
        activeAbilityTagsToSkipTestingValue {key=None, defaultValue {gameplayTags []}}
        gameplayAbilityTagContainer {key=None, defaultValue {gameplayTags [{tagName=NPC.Ability.NonAttack.Move.ToLand}]}}
        targetValue {baseClass=&Actor, defaultValue=null, key=SpawnSpot}
       }
      }]
     properties {waitTime {defaultValue=1}}
    }
   ]
   decorators [
    Has Gameplay Ability {
     class=FortBTDecorator_GameplayAbility_HasGameplayAbility
     properties {gameplayAbilityTagContainer {defaultValue {gameplayTags [{tagName=NPC.Ability.NonAttack.Move.ToLand}]}}}
    }
    Is Swimming {
     class=BTDecorator_GameplayTagQuery
     properties {
      flowAbortMode=LowerPriority
      nodeName=Is Swimming
      gameplayTagQuery {autoDescription=" ANY( MovementMode.SurfaceSwimming )", queryTokenStream [0,1,1,1,0], tagDictionary [{tagName=MovementMode.SurfaceSwimming}]}
     }
    }
   ]
   properties {
    children [
     {
      childComposite=null
      childTask=&self:FortBTTask_ExecuteGameplayAbility_19
      decorators [&self:FortBTDecorator_GameplayAbility_CanActivate_13]
     }
     {
      childComposite=null
      childTask=&self:BTTask_Wait_14
      decorators [&self:FortBTDecorator_GameplayAbility_IsOnCooldown_6]
     }
    ]
    services []
   }
   services []
  }
  Sequence {
   class=BTComposite_Sequence
   children [
    Wait {class=BTTask_Wait, properties {randomDeviation {defaultValue=10}, waitTime {defaultValue=0}}}
    GameplayAbility {
     class=FortBTTask_ExecuteGameplayAbility
     decorators [Can Activate Gameplay Ability {
       class=FortBTDecorator_GameplayAbility_CanActivate
       properties {
        flowAbortMode=None
        gameplayAbilityTagContainer {defaultValue {gameplayTags [{tagName=NPC.Ability.NonAttack.Wander.Freely}]}}
        targetValue {key=GoalActor}
       }
      }]
     properties {nodeName=GameplayAbility, gameplayAbilityTagContainer {defaultValue {gameplayTags [{tagName=NPC.Ability.NonAttack.Wander.Freely}]}}}
    }
   ]
   decorators [Has Gameplay Ability {
     class=FortBTDecorator_GameplayAbility_HasGameplayAbility
     properties {gameplayAbilityTagContainer {defaultValue {gameplayTags [{tagName=NPC.Ability.NonAttack.Wander.Freely}]}}}
    }]
   properties {children [
     {childTask=&self:BTTask_Wait_4}
     {
      childTask=&self:FortBTTask_ExecuteGameplayAbility_0
      decorators [&self:FortBTDecorator_GameplayAbility_CanActivate_9]
     }
    ]}
  }
  Wait {class=BTTask_Wait, properties {waitTime {defaultValue=1}}}
 ]
 decorators []
 properties {
  children [
   {
    childComposite=&self:BTComposite_Sequence_1
    childTask=null
    decorators [&self:BTDecorator_GameplayTagQuery_1]
   }
   {
    childComposite=null
    childTask=&self:BTTask_Wait_17
    decorators [&self:BTDecorator_GameplayTagQuery_0]
   }
   {
    childComposite=&self:BTComposite_Selector_3
    childTask=null
    decoratorOps [{number=3, operation=And}, {number=0}, {number=1}, {number=2, operation=Or}, {number=2}, {number=3}]
    decorators [
     &self:BTDecorator_Blackboard_0
     &self:FortBTDecorator_IsGoalPawn_1
     &self:FortBTDecorator_GameplayAbility_CompareDistance_0
     &self:BTDecorator_GameplayTagQuery_4
    ]
   }
   {
    childComposite=&self:BTComposite_Selector_18
    childTask=null
    decorators [&self:BTDecorator_Blackboard_6]
   }
   {
    childComposite=&self:BTComposite_Selector_34
    childTask=null
    decorators [
     &self:FortBTDecorator_GameplayAbility_HasGameplayAbility_24
     &self:BTDecorator_GameplayTagQuery_2
    ]
   }
   {
    childComposite=&self:BTComposite_Sequence_5
    childTask=null
    decorators [&self:FortBTDecorator_GameplayAbility_HasGameplayAbility_10]
   }
   {childComposite=null, childTask=&self:BTTask_Wait_0, decorators []}
  ]
  services []
 }
 services []
}
