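# Behavior tree for rideable terrestrial creature NPCs, driving them through tag-identified gameplay abilities.
# A priority selector with observer aborts covers being ridden, suspension, ping commands, fleeing danger, energy exhaustion and getting up, distant sleep, then intro, frustration, pickup and melee behavior.
# While ridden, an EQS service picks food as GoalActor for mounted grabs within a distance band and cone.
# Ping commands and sleep run as subtrees.
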BehaviorTree /Game/Athena/AI/BehaviorTrees/BT_Terrestrial
blackboard=/Game/Athena/AI/Blackboards/FortAIBB_NPC_Default
root {
 name=Selector
 class=BTComposite_Selector
 children [
  Selector {
   class=BTComposite_Selector
   children [
    Selector {
     class=BTComposite_Selector
     children [
      GameplayAbility {
       class=FortBTTask_ExecuteGameplayAbility
       decorators [Can Activate Gameplay Ability - Grab Pickup {
         class=FortBTDecorator_GameplayAbility_CanActivate
         properties {
          flowAbortMode=LowerPriority
          nodeName=Can Activate Gameplay Ability - Grab Pickup
          gameplayAbilityTagContainer {defaultValue {gameplayTags [{tagName=NPC.Ability.NonAttack.GrabPickup.Mounted}]}}
          targetValue {key=GoalActor}
         }
        }]
       properties {nodeName=GameplayAbility, gameplayAbilityTagContainer {defaultValue {gameplayTags [{tagName=NPC.Ability.NonAttack.GrabPickup.Mounted}]}}}
      }
      Wait {class=BTTask_Wait, properties {waitTime {defaultValue=10000000}}}
     ]
     decorators [
      Has GoalActor? {
       class=BTDecorator_Blackboard
       properties {
        arithmeticOperation=Equal
        intValue=0
        nodeName=Has GoalActor?
        notifyObserver=ValueChange
        stringValue=""
        blackboardKey {selectedKeyName=GoalActor}
       }
      }
      Has GrabPickup Ability {
       class=FortBTDecorator_GameplayAbility_HasGameplayAbility
       properties {
        nodeName=Has GrabPickup Ability
        gameplayAbilityTagContainer {defaultValue {gameplayTags [{tagName=NPC.Ability.NonAttack.GrabPickup.Mounted}]}}
       }
      }
      Close enough {
       class=FortBTDecorator_DistanceBetween
       properties {
        flowAbortMode=LowerPriority
        nodeName=Close enough
        operator=LessOrEqual
        blackboardKeyB {selectedKeyName=GoalActor}
        distanceCalculationUpdateRate {defaultValue=0.1}
        specifiedDistance {defaultValue=250}
       }
      }
      But not too close {
       class=FortBTDecorator_DistanceBetween
       properties {
        flowAbortMode=LowerPriority
        nodeName=But not too close
        operator=Greater
        blackboardKeyB {selectedKeyName=GoalActor}
        distanceCalculationUpdateRate {defaultValue=0.1}
        specifiedDistance {defaultValue=50}
       }
      }
      Cone Check {
       class=BTDecorator_ConeCheck
       properties {
        bInverseCondition=false
        flowAbortMode=LowerPriority
        nodeName=Cone Check
        coneDirection {bNoneIsAllowedValue=true, selectedKeyName=None}
        coneHalfAngle {defaultValue=22.5, key=None}
        coneOrigin {bNoneIsAllowedValue=false, selectedKeyName=SelfActor}
        observed {bNoneIsAllowedValue=false, selectedKeyName=GoalActor}
       }
      }
     ]
     properties {
      children [
       {
        childComposite=null
        childTask=&self:FortBTTask_ExecuteGameplayAbility_27
        decorators [&self:FortBTDecorator_GameplayAbility_CanActivate_24]
       }
       {childComposite=null, childTask=&self:BTTask_Wait_21, decorators []}
      ]
      services []
     }
     services []
    }
    Wait {class=BTTask_Wait, properties {waitTime {defaultValue=10000000}}}
   ]
   decorators [Is Being Ridden {
     class=BTDecorator_GameplayTagQuery
     properties {
      flowAbortMode=Both
      nodeName=Is Being Ridden
      gameplayTagQuery {
       autoDescription=" ANY( Granted.Riding.Creature.IsBeingRidden )"
       queryTokenStream [0,1,1,1,0]
       tagDictionary [{tagName=Granted.Riding.Creature.IsBeingRidden}]
      }
     }
    }]
   properties {
    children [
     {
      childComposite=&self:BTComposite_Selector_59
      childTask=null
      decorators [
       &self:BTDecorator_Blackboard_0
       &self:FortBTDecorator_GameplayAbility_HasGameplayAbility_36
       &self:FortBTDecorator_DistanceBetween_2
       &self:FortBTDecorator_DistanceBetween_8
       &self:BTDecorator_ConeCheck_1
      ]
     }
     {childComposite=null, childTask=&self:BTTask_Wait_18, decorators []}
    ]
    services [&self:BTService_RunEQS_6]
   }
   services [Run EQS query on a regular basis {
     class=BTService_RunEQS
     properties {
      bCallTickOnSearchStart=false
      bRestartTimerOnEachActivation=false
      bUpdateBBOnFail=false
      interval=1
      nodeName=Run EQS query on a regular basis
      randomDeviation=0
      blackboardKey {bNoneIsAllowedValue=false, selectedKeyName=GoalActor}
      eQSRequest {
       bUseBBKeyForQueryTemplate=false
       queryTemplate=/Game/Athena/AI/EQS/Behaviors/EQS_Riding_FoodItemSelection
       runMode=SingleResult
       eQSQueryBlackboardKey {bNoneIsAllowedValue=false, selectedKeyName=None}
       queryConfig []
      }
     }
    }]
  }
  Wait {
   class=BTTask_Wait
   decorators [Tractor beam or suspended in air {
     class=BTDecorator_GameplayTagQuery
     properties {
      flowAbortMode=Both
      nodeName=Tractor beam or suspended in air
      gameplayTagQuery {
       autoDescription=" ANY( Gameplay.Status.SuspendedInAir, Gameplay.Status.TractorBeam )"
       queryTokenStream [0,1,1,2,0,1]
       tagDictionary [{tagName=Gameplay.Status.SuspendedInAir}, {tagName=Gameplay.Status.TractorBeam}]
      }
     }
    }]
   properties {waitTime {defaultValue=100000}}
  }
  Ping Commands {
   class=BTTask_RunBehavior
   decorators [Valid Ping Command? {
     class=BTDecorator_Blackboard
     properties {arithmeticOperation=Greater, intValue=0, nodeName=Valid Ping Command?, stringValue=None, blackboardKey {selectedKeyName=AI_Ping_Command}}
    }]
   properties {behaviorAsset=&/BT_TerrestrialCommands, nodeName=Ping Commands}
  }
  Selector {
   class=BTComposite_Selector
   children [GameplayAbility {
     class=FortBTTask_ExecuteGameplayAbility
     decorators [
      Has Gameplay Ability {
       class=FortBTDecorator_GameplayAbility_HasGameplayAbility
       properties {gameplayAbilityTagContainer {defaultValue {gameplayTags [{tagName=NPC.Ability.NonAttack.Avoid.Danger.Primary}]}}}
      }
      Can Activate Gameplay Ability {
       class=FortBTDecorator_GameplayAbility_CanActivate
       properties {
        flowAbortMode=None
        gameplayAbilityTagContainer {defaultValue {gameplayTags [{tagName=NPC.Ability.NonAttack.Avoid.Danger.Primary}]}}
        targetValue {key=SpawnSpot}
       }
      }
     ]
     properties {nodeName=GameplayAbility, gameplayAbilityTagContainer {defaultValue {gameplayTags [{tagName=NPC.Ability.NonAttack.Avoid.Danger.Primary}]}}}
    }]
   decorators [Gameplay Tag Conditions that allow behavior {
     class=BTDecorator_GameplayTagQuery
     properties {
      flowAbortMode=Both
      nodeName=Gameplay Tag Conditions that allow behavior
      gameplayTagQuery {
       autoDescription=" ALL(  ANY( Curie.State.ElementalStatus.Burning, Curie.State.ElementalStatus.Electrocution, Athena.AI.EnvironmentalDanger.Flee ),  NONE( NPC.Status.Frustration.DueTo.Movement ) )"
       queryTokenStream [0,1,5,2,1,3,0,1,2,3,1,3]
       tagDictionary [
        {tagName=Curie.State.ElementalStatus.Burning}
        {tagName=Curie.State.ElementalStatus.Electrocution}
        {tagName=Athena.AI.EnvironmentalDanger.Flee}
        {tagName=NPC.Status.Frustration.DueTo.Movement}
       ]
      }
     }
    }]
   properties {
    children [{
      childComposite=null
      childTask=&self:FortBTTask_ExecuteGameplayAbility_12
      decorators [
       &self:FortBTDecorator_GameplayAbility_HasGameplayAbility_15
       &self:FortBTDecorator_GameplayAbility_CanActivate_14
      ]
     }]
    services []
   }
   services []
  }
  Selector {
   class=BTComposite_Selector
   children [
    GameplayAbility {
     class=FortBTTask_ExecuteGameplayAbility
     decorators [
      Is Surface Swimming {
       class=BTDecorator_GameplayTagQuery
       properties {
        flowAbortMode=LowerPriority
        nodeName=Is Surface Swimming
        gameplayTagQuery {autoDescription=" ANY( MovementMode.SurfaceSwimming )", queryTokenStream [0,1,1,1,0], tagDictionary [{tagName=MovementMode.SurfaceSwimming}]}
       }
      }
      Can Activate Gameplay Ability {
       class=FortBTDecorator_GameplayAbility_CanActivate
       properties {
        flowAbortMode=LowerPriority
        gameplayAbilityTagContainer {defaultValue {gameplayTags [{tagName=NPC.Ability.NonAttack.Move.ToLand}]}}
        targetValue {key=SpawnSpot}
       }
      }
     ]
     properties {nodeName=GameplayAbility, gameplayAbilityTagContainer {defaultValue {gameplayTags [{tagName=NPC.Ability.NonAttack.Move.ToLand}]}}}
    }
    Selector {
     class=BTComposite_Selector
     children [
      GameplayAbility {
       class=FortBTTask_ExecuteGameplayAbility
       decorators [Is not layed down {
         class=BTDecorator_GameplayTagQuery
         properties {
          flowAbortMode=LowerPriority
          nodeName=Is not layed down
          gameplayTagQuery {
           autoDescription=" NONE( NPC.Ability.NonAttack.Exhausted.IsLayedDown )"
           queryTokenStream [0,1,3,1,0]
           tagDictionary [{tagName=NPC.Ability.NonAttack.Exhausted.IsLayedDown}]
          }
         }
        }]
       properties {nodeName=GameplayAbility, gameplayAbilityTagContainer {defaultValue {gameplayTags [{tagName=NPC.Ability.NonAttack.GrabPickup}]}}}
      }
      GameplayAbility {
       class=FortBTTask_ExecuteGameplayAbility
       decorators [
        Is layed down {
         class=BTDecorator_GameplayTagQuery
         properties {
          flowAbortMode=LowerPriority
          nodeName=Is layed down
          gameplayTagQuery {
           autoDescription=" ANY( NPC.Ability.NonAttack.Exhausted.IsLayedDown )"
           queryTokenStream [0,1,1,1,0]
           tagDictionary [{tagName=NPC.Ability.NonAttack.Exhausted.IsLayedDown}]
          }
         }
        }
        Has Get Up Ability {
         class=FortBTDecorator_GameplayAbility_HasGameplayAbility
         properties {nodeName=Has Get Up Ability, gameplayAbilityTagContainer {defaultValue {gameplayTags [{tagName=NPC.Ability.NonAttack.Exhausted.GetUp}]}}}
        }
       ]
       properties {nodeName=GameplayAbility, gameplayAbilityTagContainer {defaultValue {gameplayTags [{tagName=NPC.Ability.NonAttack.Exhausted.GetUp}]}}}
      }
     ]
     decorators [
      Has GrabPickup Ability {
       class=FortBTDecorator_GameplayAbility_HasGameplayAbility
       properties {nodeName=Has GrabPickup Ability, gameplayAbilityTagContainer {defaultValue {gameplayTags [{tagName=NPC.Ability.NonAttack.GrabPickup}]}}}
      }
      Can Activate Gameplay Ability - Grab Pickup {
       class=FortBTDecorator_GameplayAbility_CanActivate
       properties {
        flowAbortMode=LowerPriority
        nodeName=Can Activate Gameplay Ability - Grab Pickup
        gameplayAbilityTagContainer {defaultValue {gameplayTags [{tagName=NPC.Ability.NonAttack.GrabPickup}]}}
        targetValue {key=GoalActor}
       }
      }
      HACK - Is Goal NOT Pawn - Observers below are LAZY {
       class=FortBTDecorator_IsGoalPawn
       properties {
        bInverseCondition=true
        flowAbortMode=Both
        nodeName=HACK - Is Goal NOT Pawn - Observers below are LAZY
        blackboardKey {bNoneIsAllowedValue=false, selectedKeyName=GoalActor}
       }
      }
      Is BlackBoard value of given Class {
       class=BTDecorator_IsBBEntryOfClass
       properties {
        bInverseCondition=false
        flowAbortMode=Both
        nodeName=Is BlackBoard value of given Class
        blackboardKey {bNoneIsAllowedValue=false, selectedKeyName=GoalActor}
        testClass {baseClass=&Object, defaultValue=/Script/FortniteGame.FortPickup, key=None}
       }
      }
      Is BlackBoard value of given Class {
       class=BTDecorator_IsBBEntryOfClass
       properties {
        bInverseCondition=false
        flowAbortMode=Both
        nodeName=Is BlackBoard value of given Class
        blackboardKey {bNoneIsAllowedValue=false, selectedKeyName=GoalActor}
        testClass {baseClass=&Object, defaultValue=/Script/FortniteGame.FortProjectileBase, key=None}
       }
      }
      Has GoalActor? To Abort Self for goals likely to disappear {
       class=BTDecorator_Blackboard
       properties {
        arithmeticOperation=Equal
        intValue=0
        nodeName=Has GoalActor? To Abort Self for goals likely to disappear
        notifyObserver=ValueChange
        stringValue=""
        blackboardKey {selectedKeyName=GoalActor}
       }
      }
      Cone Check {
       class=BTDecorator_ConeCheck
       properties {
        bInverseCondition=false
        flowAbortMode=LowerPriority
        nodeName=Cone Check
        coneDirection {bNoneIsAllowedValue=true, selectedKeyName=None}
        coneHalfAngle {defaultValue=25, key=None}
        coneOrigin {bNoneIsAllowedValue=false, selectedKeyName=SelfActor}
        observed {bNoneIsAllowedValue=false, selectedKeyName=GoalActor}
       }
      }
      Distance Between {
       class=FortBTDecorator_DistanceBetween
       properties {
        flowAbortMode=LowerPriority
        nodeName=Distance Between
        operator=LessOrEqual
        blackboardKeyB {selectedKeyName=GoalActor}
        specifiedDistance {defaultValue=200}
       }
      }
     ]
     properties {
      children [
       {
        childComposite=null
        childTask=&self:FortBTTask_ExecuteGameplayAbility_30
        decorators [&self:BTDecorator_GameplayTagQuery_6]
       }
       {
        childComposite=null
        childTask=&self:FortBTTask_ExecuteGameplayAbility_32
        decorators [
         &self:BTDecorator_GameplayTagQuery_5
         &self:FortBTDecorator_GameplayAbility_HasGameplayAbility_9
        ]
       }
      ]
      services []
     }
     services []
    }
    GameplayAbility {
     class=FortBTTask_ExecuteGameplayAbility
     decorators [
      Is not layed down {
       class=BTDecorator_GameplayTagQuery
       properties {
        flowAbortMode=LowerPriority
        nodeName=Is not layed down
        gameplayTagQuery {
         autoDescription=" NONE( NPC.Ability.NonAttack.Exhausted.IsLayedDown )"
         queryTokenStream [0,1,3,1,0]
         tagDictionary [{tagName=NPC.Ability.NonAttack.Exhausted.IsLayedDown}]
        }
       }
      }
      Has Lay Down Ability {
       class=FortBTDecorator_GameplayAbility_HasGameplayAbility
       properties {nodeName=Has Lay Down Ability, gameplayAbilityTagContainer {defaultValue {gameplayTags [{tagName=NPC.Ability.NonAttack.Exhausted.LayDown}]}}}
      }
      Can Activate Gameplay Ability - Lay Down {
       class=FortBTDecorator_GameplayAbility_CanActivate
       properties {
        flowAbortMode=LowerPriority
        nodeName=Can Activate Gameplay Ability - Lay Down
        gameplayAbilityTagContainer {defaultValue {gameplayTags [{tagName=NPC.Ability.NonAttack.Exhausted.LayDown}]}}
        targetValue {key=GoalActor}
       }
      }
     ]
     properties {nodeName=GameplayAbility, gameplayAbilityTagContainer {defaultValue {gameplayTags [{tagName=NPC.Ability.NonAttack.Exhausted.LayDown}]}}}
    }
    Wait {class=BTTask_Wait, properties {waitTime {defaultValue=5}}}
   ]
   decorators [
    Is Energy Depleted {
     class=BTDecorator_GameplayTagQuery
     properties {
      flowAbortMode=Both
      nodeName=Is Energy Depleted
      gameplayTagQuery {
       autoDescription=" ANY( Granted.Riding.Creature.Energy.Depleted )"
       queryTokenStream [0,1,1,1,0]
       tagDictionary [{tagName=Granted.Riding.Creature.Energy.Depleted}]
      }
     }
    }
    Has Lay Down Ability {
     class=FortBTDecorator_GameplayAbility_HasGameplayAbility
     properties {nodeName=Has Lay Down Ability, gameplayAbilityTagContainer {defaultValue {gameplayTags [{tagName=NPC.Ability.NonAttack.Exhausted.LayDown}]}}}
    }
   ]
   properties {
    children [
     {
      childComposite=null
      childTask=&self:FortBTTask_ExecuteGameplayAbility_25
      decorators [
       &self:BTDecorator_GameplayTagQuery_2
       &self:FortBTDecorator_GameplayAbility_CanActivate_5
      ]
     }
     {
      childComposite=&self:BTComposite_Selector_51
      childTask=null
      decoratorOps [
       {number=7, operation=And}
       {number=0}
       {number=1}
       {number=2}
       {number=2, operation=Or}
       {number=3}
       {number=4}
       {number=5}
       {number=6}
       {number=7}
      ]
      decorators [
       &self:FortBTDecorator_GameplayAbility_HasGameplayAbility_8
       &self:FortBTDecorator_GameplayAbility_CanActivate_6
       &self:FortBTDecorator_IsGoalPawn_1
       &self:BTDecorator_IsBBEntryOfClass_0
       &self:BTDecorator_IsBBEntryOfClass_1
       &self:BTDecorator_Blackboard_5
       &self:BTDecorator_ConeCheck_2
       &self:FortBTDecorator_DistanceBetween_3
      ]
     }
     {
      childComposite=null
      childTask=&self:FortBTTask_ExecuteGameplayAbility_26
      decorators [
       &self:BTDecorator_GameplayTagQuery_0
       &self:FortBTDecorator_GameplayAbility_HasGameplayAbility_24
       &self:FortBTDecorator_GameplayAbility_CanActivate_19
      ]
     }
     {childComposite=null, childTask=&self:BTTask_Wait_16, decorators []}
    ]
    services []
   }
   services []
  }
  GameplayAbility {
   class=FortBTTask_ExecuteGameplayAbility
   decorators [
    Is Energy Not Depleted And Laying Down {
     class=BTDecorator_GameplayTagQuery
     properties {
      flowAbortMode=LowerPriority
      nodeName=Is Energy Not Depleted And Laying Down
      gameplayTagQuery {
       autoDescription=" ALL(  NONE( GameplayEvent.Riding.Player.EnergyDepleted ),  ANY( NPC.Ability.NonAttack.Exhausted.IsLayedDown ) )"
       queryTokenStream [0,1,5,2,3,1,0,1,1,1]
       tagDictionary [{tagName=GameplayEvent.Riding.Player.EnergyDepleted}, {tagName=NPC.Ability.NonAttack.Exhausted.IsLayedDown}]
      }
     }
    }
    Has Get Up Ability {
     class=FortBTDecorator_GameplayAbility_HasGameplayAbility
     properties {nodeName=Has Get Up Ability, gameplayAbilityTagContainer {defaultValue {gameplayTags [{tagName=NPC.Ability.NonAttack.Exhausted.GetUp}]}}}
    }
   ]
   properties {nodeName=GameplayAbility, gameplayAbilityTagContainer {defaultValue {gameplayTags [{tagName=NPC.Ability.NonAttack.Exhausted.GetUp}]}}}
  }
  Selector {
   class=BTComposite_Selector
   children [Run Behavior {class=BTTask_RunBehavior, decorators [], properties {behaviorAsset=/Game/AI/BehaviorTrees/Subtrees/BT_UrgentSleep, nodeName=Run Behavior}}]
   decorators [AI Far enough from players to sleep and not immobilized {
     class=BTDecorator_GameplayTagQuery
     properties {
      flowAbortMode=Both
      nodeName=AI Far enough from players to sleep and not immobilized
      gameplayTagQuery {
       autoDescription=" ALL(  ANY( NPC.Status.FarAwayFromPlayers ),  NONE( Gameplay.Status.Immobilize ) )"
       queryTokenStream [0,1,5,2,1,1,0,3,1,1]
       tagDictionary [{tagName=NPC.Status.FarAwayFromPlayers}, {tagName=Gameplay.Status.Immobilize}]
      }
     }
    }]
   properties {children [{childComposite=null, childTask=&self:BTTask_RunBehavior_0, decorators []}], services []}
   services []
  }
  Selector {
   class=BTComposite_Selector
   children [
    Selector {
     class=BTComposite_Selector
     children [Selector {
       class=BTComposite_Selector
       children [GameplayAbility {
         class=FortBTTask_ExecuteGameplayAbility
         decorators [Compare Gameplay Ability Distances {
           class=FortBTDecorator_GameplayAbility_CompareDistance
           properties {
            bInverseCondition=false
            flowAbortMode=LowerPriority
            nodeName=Compare Gameplay Ability Distances
            activeAbilityTagsToSkipTestingValue {key=None, defaultValue {gameplayTags []}}
            distanceComparisons [{
              addedOffset=0
              bUseAddedOffset=false
              bUseOverriddenValue=false
              comparisonType=ThreeDimensions
              operator=Less
              overriddenValue=0
              distanceDataTags {gameplayTags [{tagName=NPC.Ability.MaxPreferredRange}]}
             }]
            gameplayAbilityTagContainer {key=None, defaultValue {gameplayTags [{tagName=NPC.Ability.NonAttack.PlayMontage.Intro}]}}
            targetValue {baseClass=&Actor, defaultValue=null, key=GoalActor}
           }
          }]
         properties {nodeName=GameplayAbility, gameplayAbilityTagContainer {defaultValue {gameplayTags [{tagName=NPC.Ability.NonAttack.PlayMontage.Intro}]}}}
        }]
       decorators [
        Has Gameplay Ability {
         class=FortBTDecorator_GameplayAbility_HasGameplayAbility
         properties {gameplayAbilityTagContainer {defaultValue {gameplayTags [{tagName=NPC.Ability.NonAttack.PlayMontage.Intro}]}}}
        }
        Is intro animation pending {
         class=BTDecorator_GameplayTagQuery
         properties {
          flowAbortMode=None
          nodeName=Is intro animation pending
          gameplayTagQuery {
           autoDescription=" ANY( NPC.Behavior.Phase.Intro.Animation.Pending )"
           queryTokenStream [0,1,1,1,0]
           tagDictionary [{tagName=NPC.Behavior.Phase.Intro.Animation.Pending}]
          }
         }
        }
       ]
       properties {
        children [{
          childComposite=null
          childTask=&self:FortBTTask_ExecuteGameplayAbility_0
          decorators [&self:FortBTDecorator_GameplayAbility_CompareDistance_19]
         }]
        services []
       }
       services []
      }]
     decorators [
      Is In Intro Phase {
       class=BTDecorator_GameplayTagQuery
       properties {
        flowAbortMode=LowerPriority
        nodeName=Is In Intro Phase
        gameplayTagQuery {
         autoDescription=" ANY( NPC.Behavior.Phase.Intro.BehaviorActive )"
         queryTokenStream [0,1,1,1,0]
         tagDictionary [{tagName=NPC.Behavior.Phase.Intro.BehaviorActive}]
        }
       }
      }
      Is Goal Pawn {
       class=FortBTDecorator_IsGoalPawn
       properties {
        bInverseCondition=false
        flowAbortMode=LowerPriority
        nodeName=Is Goal Pawn
        blackboardKey {bNoneIsAllowedValue=false, selectedKeyName=GoalActor}
       }
      }
     ]
     properties {
      children [{
        childComposite=&self:BTComposite_Selector_19
        childTask=null
        decorators [
         &self:FortBTDecorator_GameplayAbility_HasGameplayAbility_6
         &self:BTDecorator_GameplayTagQuery_15
        ]
       }]
      services []
     }
     services []
    }
    Selector {
     class=BTComposite_Selector
     children [
      Selector {
       class=BTComposite_Selector
       children [
        Selector {
         class=BTComposite_Selector
         children [Sequence {
           class=BTComposite_Sequence
           children [
            Wait {class=BTTask_Wait, properties {waitTime {defaultValue=0.5}}}
            GameplayAbility {
             class=FortBTTask_ExecuteGameplayAbility
             decorators []
             properties {
              nodeName=GameplayAbility
              gameplayAbilityTagContainer {defaultValue {gameplayTags [{tagName=NPC.Ability.NonAttack.Frustration.DueTo.Movement}]}}
             }
            }
           ]
           decorators []
           properties {children [
             {childTask=&self:BTTask_Wait_10}
             {childTask=&self:FortBTTask_ExecuteGameplayAbility_14}
            ]}
          }]
         decorators [
          Is NOT Falling? {
           class=BTDecorator_GameplayTagQuery
           properties {
            flowAbortMode=LowerPriority
            nodeName=Is NOT Falling?
            gameplayTagQuery {autoDescription=" NONE( MovementMode.Falling )", queryTokenStream [0,1,3,1,0], tagDictionary [{tagName=MovementMode.Falling}]}
           }
          }
          Has Gameplay Ability {
           class=FortBTDecorator_GameplayAbility_HasGameplayAbility
           properties {gameplayAbilityTagContainer {defaultValue {gameplayTags [{tagName=NPC.Ability.NonAttack.Frustration.DueTo.Movement}]}}}
          }
          NPC.Status.Frustration.DueTo.Movement {
           class=BTDecorator_CheckGameplayTagsOnActor
           properties {
            bInverseCondition=false
            flowAbortMode=None
            nodeName=NPC.Status.Frustration.DueTo.Movement
            tagsToMatch=Any
            actorToCheck {bNoneIsAllowedValue=false, selectedKeyName=SelfActor}
            gameplayTags {gameplayTags [{tagName=NPC.Status.Frustration.DueTo.Movement}]}
           }
          }
         ]
         properties {children [{childComposite=&self:BTComposite_Sequence_2, childTask=null, decorators []}], services []}
         services []
        }
        Selector {
         class=BTComposite_Selector
         children [
          GameplayAbility {
           class=FortBTTask_ExecuteGameplayAbility
           decorators [MaxActualRange {
             class=FortBTDecorator_GameplayAbility_CompareDistance
             properties {
              bInverseCondition=false
              flowAbortMode=None
              nodeName=MaxActualRange
              activeAbilityTagsToSkipTestingValue {key=None, defaultValue {gameplayTags []}}
              distanceComparisons [{
                addedOffset=0
                bUseAddedOffset=false
                bUseOverriddenValue=false
                comparisonType=ThreeDimensions
                operator=Less
                overriddenValue=0
                distanceDataTags {gameplayTags [{tagName=NPC.Ability.MaxActualRange}]}
               }]
              gameplayAbilityTagContainer {key=None, defaultValue {gameplayTags [{tagName=NPC.Ability.NonAttack.GrabPickup}]}}
              targetValue {baseClass=&Actor, defaultValue=null, key=GoalActor}
             }
            }]
           properties {nodeName=GameplayAbility, gameplayAbilityTagContainer {defaultValue {gameplayTags [{tagName=NPC.Ability.NonAttack.GrabPickup}]}}}
          }
          Selector {
           class=BTComposite_Selector
           children [
            GameplayAbility {
             class=FortBTTask_ExecuteGameplayAbility
             decorators [Does path exist {
               class=BTDecorator_DoesPathExist
               properties {
                bInverseCondition=false
                flowAbortMode=None
                nodeName=Does path exist
                blackboardKeyA {bNoneIsAllowedValue=false, selectedKeyName=SelfActor}
                blackboardKeyB {bNoneIsAllowedValue=false, selectedKeyName=GoalActor}
                filterClass {baseClass=&NavigationQueryFilter, defaultValue=null, key=None}
                pathQueryType {defaultValue=2, enumType=/Script/AIModule.EPathExistanceQueryType, key=None, nativeEnumTypeName=EPathExistanceQueryType}
               }
              }]
             properties {nodeName=GameplayAbility, gameplayAbilityTagContainer {defaultValue {gameplayTags [{tagName=NPC.Ability.NonAttack.Move.ToPickup}]}}}
            }
            Set Frustration Discouragement {
             class=FortBTTask_SetFrustrationDiscouragement
             children []
             decorators []
             properties {bIgnoreRestartSelf=false, discouragementDuration=10, nodeName=Set Frustration Discouragement}
             services []
            }
           ]
           decorators [
            Has Move To Pickup Ability {
             class=FortBTDecorator_GameplayAbility_HasGameplayAbility
             properties {
              nodeName=Has Move To Pickup Ability
              gameplayAbilityTagContainer {defaultValue {gameplayTags [{tagName=NPC.Ability.NonAttack.Move.ToPickup}]}}
             }
            }
            Can Activate Gameplay Ability - Move To Pickup {
             class=FortBTDecorator_GameplayAbility_CanActivate
             properties {
              flowAbortMode=LowerPriority
              nodeName=Can Activate Gameplay Ability - Move To Pickup
              gameplayAbilityTagContainer {defaultValue {gameplayTags [{tagName=NPC.Ability.NonAttack.Move.ToPickup}]}}
              targetValue {key=GoalActor}
             }
            }
           ]
           properties {
            children [
             {
              childComposite=null
              childTask=&self:FortBTTask_ExecuteGameplayAbility_5
              decorators [&self:BTDecorator_DoesPathExist_5]
             }
             {childComposite=null, childTask=&self:FortBTTask_SetFrustrationDiscouragement_0, decorators []}
            ]
            services []
           }
           services []
          }
         ]
         decorators [
          Has GrabPickup Ability {
           class=FortBTDecorator_GameplayAbility_HasGameplayAbility
           properties {nodeName=Has GrabPickup Ability, gameplayAbilityTagContainer {defaultValue {gameplayTags [{tagName=NPC.Ability.NonAttack.GrabPickup}]}}}
          }
          Can Activate Gameplay Ability - Grab Pickup {
           class=FortBTDecorator_GameplayAbility_CanActivate
           properties {
            flowAbortMode=LowerPriority
            nodeName=Can Activate Gameplay Ability - Grab Pickup
            gameplayAbilityTagContainer {defaultValue {gameplayTags [{tagName=NPC.Ability.NonAttack.GrabPickup}]}}
            targetValue {key=GoalActor}
           }
          }
          HACK - Is Goal NOT Pawn - Observers below are LAZY {
           class=FortBTDecorator_IsGoalPawn
           properties {
            bInverseCondition=true
            flowAbortMode=Both
            nodeName=HACK - Is Goal NOT Pawn - Observers below are LAZY
            blackboardKey {bNoneIsAllowedValue=false, selectedKeyName=GoalActor}
           }
          }
          Is BlackBoard value of given Class {
           class=BTDecorator_IsBBEntryOfClass
           properties {
            bInverseCondition=false
            flowAbortMode=Both
            nodeName=Is BlackBoard value of given Class
            blackboardKey {bNoneIsAllowedValue=false, selectedKeyName=GoalActor}
            testClass {baseClass=&Object, defaultValue=&FortPickup, key=None}
           }
          }
          Is BlackBoard value of given Class {
           class=BTDecorator_IsBBEntryOfClass
           properties {
            bInverseCondition=false
            flowAbortMode=Both
            nodeName=Is BlackBoard value of given Class
            blackboardKey {bNoneIsAllowedValue=false, selectedKeyName=GoalActor}
            testClass {baseClass=&Object, defaultValue=&FortProjectileBase, key=None}
           }
          }
          Has GoalActor? To Abort Self for goals likely to disappear {
           class=BTDecorator_Blackboard
           properties {
            arithmeticOperation=Equal
            intValue=0
            nodeName=Has GoalActor? To Abort Self for goals likely to disappear
            notifyObserver=ValueChange
            stringValue=""
            blackboardKey {selectedKeyName=GoalActor}
           }
          }
         ]
         properties {
          children [
           {
            childComposite=null
            childTask=&self:FortBTTask_ExecuteGameplayAbility_4
            decorators [&self:FortBTDecorator_GameplayAbility_CompareDistance_23]
           }
           {
            childComposite=&self:BTComposite_Selector_47
            childTask=null
            decorators [
             &self:FortBTDecorator_GameplayAbility_HasGameplayAbility_34
             &self:FortBTDecorator_GameplayAbility_CanActivate_26
            ]
           }
          ]
          services []
         }
         services []
        }
       ]
       decorators [Has ANY Attack Ability? {
         class=FortBTDecorator_GameplayAbility_HasGameplayAbility
         properties {nodeName=Has ANY Attack Ability?, gameplayAbilityTagContainer {defaultValue {gameplayTags [{tagName=NPC.Ability.Attack}]}}}
        }]
       properties {
        children [
         {
          childComposite=&self:BTComposite_Selector_26
          childTask=null
          decorators [
           &self:BTDecorator_GameplayTagQuery_9
           &self:FortBTDecorator_GameplayAbility_HasGameplayAbility_21
           &self:BTDecorator_CheckGameplayTagsOnActor_11
          ]
         }
         {
          childComposite=&self:BTComposite_Selector_32
          childTask=null
          decoratorOps [{number=5, operation=And}, {number=0}, {number=1}, {number=2}, {number=2, operation=Or}, {number=3}, {number=4}, {number=5}]
          decorators [
           &self:FortBTDecorator_GameplayAbility_HasGameplayAbility_33
           &self:FortBTDecorator_GameplayAbility_CanActivate_25
           &self:FortBTDecorator_IsGoalPawn_8
           &self:BTDecorator_IsBBEntryOfClass_14
           &self:BTDecorator_IsBBEntryOfClass_15
           &self:BTDecorator_Blackboard_11
          ]
         }
        ]
        services []
       }
       services []
      }
      Selector {
       class=BTComposite_Selector
       children [
        Selector {
         class=BTComposite_Selector
         children [Selector {
           class=BTComposite_Selector
           children [Selector {
             class=BTComposite_Selector
             children [Selector {
               class=BTComposite_Selector
               children [Selector {
                 class=BTComposite_Selector
                 children [Selector {
                   class=BTComposite_Selector
                   children [Selector {
                     class=BTComposite_Selector
                     children [
                      GameplayAbility {
                       class=FortBTTask_ExecuteGameplayAbility
                       decorators [
                        TimeLimit {
                         class=BTDecorator_TimeLimit
                         properties {bInverseCondition=false, flowAbortMode=Self, nodeName=TimeLimit, timeLimit {defaultValue=1, key=None}}
                        }
                        Compare Gameplay Ability Distances {
                         class=FortBTDecorator_GameplayAbility_CompareDistance
                         properties {
                          bInverseCondition=false
                          flowAbortMode=None
                          nodeName=Compare Gameplay Ability Distances
                          activeAbilityTagsToSkipTestingValue {key=None, defaultValue {gameplayTags []}}
                          distanceComparisons [{
                            addedOffset=0
                            bUseAddedOffset=false
                            bUseOverriddenValue=false
                            comparisonType=ThreeDimensions
                            operator=Less
                            overriddenValue=0
                            distanceDataTags {gameplayTags [{tagName=NPC.Ability.MinPreferredRange}]}
                           }]
                          gameplayAbilityTagContainer {key=None, defaultValue {gameplayTags [{tagName=NPC.Ability.Attack.Melee.Primary}]}}
                          targetValue {baseClass=&Actor, defaultValue=null, key=GoalActor}
                         }
                        }
                        Has Gameplay Ability {
                         class=FortBTDecorator_GameplayAbility_HasGameplayAbility
                         properties {gameplayAbilityTagContainer {defaultValue {gameplayTags [{tagName=NPC.Ability.NonAttack.Reposition.MinPreferredRange.Melee.Backup}]}}}
                        }
                        Can Activate Gameplay Ability {
                         class=FortBTDecorator_GameplayAbility_CanActivate
                         properties {
                          flowAbortMode=None
                          gameplayAbilityTagContainer {defaultValue {gameplayTags [{tagName=NPC.Ability.NonAttack.Reposition.MinPreferredRange.Melee.Backup}]}}
                          targetValue {key=SpawnSpot}
                         }
                        }
                       ]
                       properties {
                        nodeName=GameplayAbility
                        gameplayAbilityTagContainer {defaultValue {gameplayTags [{tagName=NPC.Ability.NonAttack.Reposition.MinPreferredRange.Melee.Backup}]}}
                       }
                      }
                      Selector {
                       class=BTComposite_Selector
                       children [
                        Rotate to face BB entry {
                         class=BTTask_RotateToFaceBBEntry
                         decorators [NOT Rotated To Attack Target {
                           class=FortBTDecorator_GameplayAbility_IsRotatedToAttackTarget
                           properties {
                            bInverseCondition=true
                            flowAbortMode=None
                            nodeName=NOT Rotated To Attack Target
                            activeAbilityTagsToSkipTestingValue {key=None, defaultValue {gameplayTags []}}
                            gameplayAbilityTagContainer {key=None, defaultValue {gameplayTags [{tagName=NPC.Ability.Attack.Melee.Primary}]}}
                            targetValue {baseClass=&Actor, defaultValue=null, key=GoalActor}
                           }
                          }]
                         properties {blackboardKey {selectedKeyName=GoalActor}, precision {defaultValue=10}}
                        }
                        Wait {class=BTTask_Wait, properties {waitTime {defaultValue=0.1}}}
                       ]
                       decorators [Could Hit if we Rotated {
                         class=FortBTDecorator_GameplayAbility_CanHitTarget
                         properties {
                          bInverseCondition=false
                          flowAbortMode=None
                          nodeName=Could Hit if we Rotated
                          activeAbilityTagsToSkipTestingValue {key=None, defaultValue {gameplayTags []}}
                          gameplayAbilityTagContainer {key=None, defaultValue {gameplayTags [{tagName=NPC.Ability.Attack.Melee.Primary}]}}
                          targetValue {baseClass=&Actor, defaultValue=null, key=GoalActor}
                          useIdealYawRotationToTargetValue {defaultValue=true, key=None}
                         }
                        }]
                       properties {
                        children [
                         {
                          childComposite=null
                          childTask=&self:BTTask_RotateToFaceBBEntry_0
                          decorators [&self:FortBTDecorator_GameplayAbility_IsRotatedToAttackTarget_0]
                         }
                         {childComposite=null, childTask=&self:BTTask_Wait_8, decorators []}
                        ]
                        services []
                       }
                       services []
                      }
                      Sequence {
                       class=BTComposite_Sequence
                       children [
                        Selector {
                         class=BTComposite_Selector
                         children [
                          GameplayAbility {
                           class=FortBTTask_ExecuteGameplayAbility
                           decorators [
                            Has Gameplay Ability {
                             class=FortBTDecorator_GameplayAbility_HasGameplayAbility
                             properties {gameplayAbilityTagContainer {defaultValue {gameplayTags [{tagName=NPC.Ability.NonAttack.Reposition.MinPreferredRange.Melee.MoveToTarget}]}}}
                            }
                            Can Activate Gameplay Ability {
                             class=FortBTDecorator_GameplayAbility_CanActivate
                             properties {
                              flowAbortMode=None
                              gameplayAbilityTagContainer {defaultValue {gameplayTags [{tagName=NPC.Ability.NonAttack.Reposition.MinPreferredRange.Melee.MoveToTarget}]}}
                              targetValue {key=SpawnSpot}
                             }
                            }
                            Is NOT Closest Nearby Pawn {
                             class=FortBTDecorator_GameplayAbility_IsClosestPawnInRange
                             properties {
                              bInverseCondition=true
                              flowAbortMode=None
                              nodeName=Is NOT Closest Nearby Pawn
                              activeAbilityTagsToSkipTestingValue {key=None, defaultValue {gameplayTags []}}
                              filterAIPawns {defaultValue=true, key=None}
                              filterPlayerPawns {defaultValue=false, key=None}
                              gameplayAbilityTagContainer {key=None, defaultValue {gameplayTags [{tagName=NPC.Ability.NonAttack.Reposition.MinPreferredRange.Melee.MoveToTarget}]}}
                              nearbyPawnDistanceValue {defaultValue=3000, key=None}
                              targetValue {baseClass=&Actor, defaultValue=null, key=GoalActor}
                             }
                            }
                           ]
                           properties {
                            nodeName=GameplayAbility
                            gameplayAbilityTagContainer {defaultValue {gameplayTags [{tagName=NPC.Ability.NonAttack.Reposition.MinPreferredRange.Melee.MoveToTarget}]}}
                           }
                          }
                          FortMoveTo {
                           class=FortBTTask_GameMoveTo
                           properties {
                            bDeimosFlavor=false
                            bFinishMoveOnOverlap=true
                            bStopAtGoal=false
                            nodeName=FortMoveTo
                            acceptableRadius {defaultValue=30}
                            bReachTestIncludesAgentRadius {defaultValue=true}
                            bReachTestIncludesGoalRadius {defaultValue=true}
                            blackboardKey {selectedKeyName=GoalActor}
                            filterClass {defaultValue=&NavFilter_NPC_AnyTraversable_C}
                            navFilterTag {tagName=NavFilter.NPC.DefaultMove}
                           }
                          }
                         ]
                         decorators []
                         properties {
                          children [
                           {
                            childComposite=null
                            childTask=&self:FortBTTask_ExecuteGameplayAbility_22
                            decorators [
                             &self:FortBTDecorator_GameplayAbility_HasGameplayAbility_29
                             &self:FortBTDecorator_GameplayAbility_CanActivate_12
                             &self:FortBTDecorator_GameplayAbility_IsClosestPawnInRange_0
                            ]
                           }
                           {childComposite=null, childTask=&self:FortBTTask_GameMoveTo_5, decorators []}
                          ]
                          services []
                         }
                         services []
                        }
                        Wait {class=BTTask_Wait, properties {waitTime {defaultValue=0.2}}}
                       ]
                       decorators [Compare Gameplay Ability Distances {
                         class=FortBTDecorator_GameplayAbility_CompareDistance
                         properties {
                          bInverseCondition=false
                          flowAbortMode=Both
                          nodeName=Compare Gameplay Ability Distances
                          activeAbilityTagsToSkipTestingValue {key=None, defaultValue {gameplayTags []}}
                          distanceComparisons [{
                            addedOffset=0
                            bUseAddedOffset=false
                            bUseOverriddenValue=false
                            comparisonType=ThreeDimensions
                            operator=Greater
                            overriddenValue=0
                            distanceDataTags {gameplayTags [{tagName=NPC.Ability.MinPreferredRange}]}
                           }]
                          gameplayAbilityTagContainer {key=None, defaultValue {gameplayTags [{tagName=NPC.Ability.Attack.Melee.Primary}]}}
                          targetValue {baseClass=&Actor, defaultValue=null, key=GoalActor}
                         }
                        }]
                       properties {children [
                         {childComposite=&self:BTComposite_Selector_54, childTask=null}
                         {childTask=&self:BTTask_Wait_2}
                        ]}
                      }
                     ]
                     decorators [
                      NO AbandonSlot tag? {
                       class=BTDecorator_GameplayTagQuery
                       properties {
                        flowAbortMode=None
                        nodeName=NO AbandonSlot tag?
                        gameplayTagQuery {
                         autoDescription=" NONE( NPC.Behavior.TargetSlots.Melee.AbandonSlot )"
                         queryTokenStream [0,1,3,1,0]
                         tagDictionary [{tagName=NPC.Behavior.TargetSlots.Melee.AbandonSlot}]
                        }
                       }
                      }
                      Does path exist {
                       class=BTDecorator_DoesPathExist
                       properties {
                        bInverseCondition=false
                        flowAbortMode=None
                        nodeName=Does path exist
                        blackboardKeyA {bNoneIsAllowedValue=false, selectedKeyName=SelfActor}
                        blackboardKeyB {bNoneIsAllowedValue=false, selectedKeyName=GoalActor}
                        filterClass {baseClass=&NavigationQueryFilter, defaultValue=null, key=None}
                        pathQueryType {defaultValue=2, enumType=&EPathExistanceQueryType, key=None, nativeEnumTypeName=EPathExistanceQueryType}
                       }
                      }
                     ]
                     properties {
                      children [
                       {
                        childComposite=null
                        childTask=&self:FortBTTask_ExecuteGameplayAbility_15
                        decorators [
                         &self:BTDecorator_TimeLimit_0
                         &self:FortBTDecorator_GameplayAbility_CompareDistance_18
                         &self:FortBTDecorator_GameplayAbility_HasGameplayAbility_20
                         &self:FortBTDecorator_GameplayAbility_CanActivate_15
                        ]
                       }
                       {
                        childComposite=&self:BTComposite_Selector_7
                        childTask=null
                        decorators [&self:FortBTDecorator_GameplayAbility_CanHitTarget_0]
                       }
                       {
                        childComposite=&self:BTComposite_Sequence_7
                        childTask=null
                        decorators [&self:FortBTDecorator_GameplayAbility_CompareDistance_17]
                       }
                      ]
                      services []
                     }
                     services []
                    }]
                   decorators [Loop {class=BTDecorator_Loop}]
                   properties {
                    children [{
                      childComposite=&self:BTComposite_Selector_17
                      childTask=null
                      decorators [
                       &self:BTDecorator_GameplayTagQuery_19
                       &self:BTDecorator_DoesPathExist_1
                      ]
                     }]
                    services []
                   }
                   services []
                  }]
                 decorators [
                  NOT Maxed {
                   class=BTDecorator_CheckGameplayTagsOnActor
                   properties {
                    bInverseCondition=true
                    flowAbortMode=None
                    nodeName=NOT Maxed
                    tagsToMatch=Any
                    actorToCheck {bNoneIsAllowedValue=false, selectedKeyName=GoalActor}
                    gameplayTags {gameplayTags [{tagName=NPC.Behavior.TargetSlots.Melee.Max}]}
                   }
                  }
                  TargetSlots Can Fit? {
                   class=TargetSlotsAvailable_C
                   properties {
                    bCheckConditionOnlyBlackBoardChanges=false
                    bInverseCondition=false
                    bShowPropertyDetails=true
                    customDescription=""
                    flowAbortMode=LowerPriority
                    nPC PawnParent=null
                    nPC PawnParentController=null
                    nodeName=TargetSlots Can Fit?
                    targetSlotType=Melee
                    nPC BlackboardKeyToTestTargetSlotsOn {bNoneIsAllowedValue=false, selectedKeyName=GoalActor}
                   }
                  }
                  MAXED {
                   class=BTDecorator_CheckGameplayTagsOnActor
                   properties {
                    bInverseCondition=false
                    flowAbortMode=None
                    nodeName=MAXED
                    tagsToMatch=Any
                    actorToCheck {bNoneIsAllowedValue=false, selectedKeyName=GoalActor}
                    gameplayTags {gameplayTags [{tagName=NPC.Behavior.TargetSlots.Melee.Max}]}
                   }
                  }
                  OVERRIDE {
                   class=BTDecorator_CheckGameplayTagsOnActor
                   properties {
                    bInverseCondition=false
                    flowAbortMode=None
                    nodeName=OVERRIDE
                    tagsToMatch=Any
                    actorToCheck {bNoneIsAllowedValue=false, selectedKeyName=SelfActor}
                    gameplayTags {gameplayTags [{tagName=NPC.Behavior.TargetSlots.Melee.OverrideSlot}]}
                   }
                  }
                  NO ABANDON {
                   class=BTDecorator_CheckGameplayTagsOnActor
                   properties {
                    bInverseCondition=true
                    flowAbortMode=None
                    nodeName=NO ABANDON
                    tagsToMatch=Any
                    actorToCheck {bNoneIsAllowedValue=false, selectedKeyName=SelfActor}
                    gameplayTags {gameplayTags [{tagName=NPC.Behavior.TargetSlots.Melee.AbandonSlot}]}
                   }
                  }
                 ]
                 properties {
                  children [{
                    childComposite=&self:BTComposite_Selector_15
                    childTask=null
                    decorators [&self:BTDecorator_Loop_2]
                   }]
                  services [&self:FortBTService_SendGameplayEvents_C_2]
                 }
                 services [Send Gameplay Events to Blackboard Key {
                   class=FortBTService_SendGameplayEvents_C
                   properties {
                    activatingPawn=null
                    activatingPawnAbilitySystemComponent=null
                    bCallTickOnSearchStart=false
                    bRestartTimerOnEachActivation=false
                    bShowEventDetails=false
                    bShowPropertyDetails=true
                    customDescription=""
                    goalActor=null
                    interval=0.5
                    nodeName=Send Gameplay Events to Blackboard Key
                    ownerActor=null
                    randomDeviation=0.1
                    targetAbilitySystemComponent=null
                    useEventTickLogic=false
                    activationEventKeysTags []
                    activationEventValuesBlackboards []
                    activationEvents {"(TagName=\"NPC.Behavior.TargetSlots.Melee.Add\")" {bNoneIsAllowedValue=false, selectedKeyName=SelfActor}}
                    deactivationEventKeysTags []
                    deactivationEvents {"(TagName=\"NPC.Behavior.TargetSlots.Melee.Remove\")" {bNoneIsAllowedValue=false, selectedKeyName=SelfActor}}
                    deactivationEventsValuesBlackboards []
                   }
                  }]
                }]
               decorators [
                Has Gameplay Ability - Target Slots Melee {
                 class=FortBTDecorator_GameplayAbility_HasGameplayAbility
                 properties {
                  nodeName=Has Gameplay Ability - Target Slots Melee
                  gameplayAbilityTagContainer {defaultValue {gameplayTags [{tagName=NPC.Ability.NonAttack.TargetSlots.Melee}]}}
                 }
                }
                Has Gameplay Ability - Melee Primary {
                 class=FortBTDecorator_GameplayAbility_HasGameplayAbility
                 properties {
                  nodeName=Has Gameplay Ability - Melee Primary
                  gameplayAbilityTagContainer {defaultValue {gameplayTags [{tagName=NPC.Ability.Attack.Melee.Primary}]}}
                 }
                }
                Within Primary Melee Approach Range? {
                 class=FortBTDecorator_GameplayAbility_CompareDistance
                 properties {
                  bInverseCondition=false
                  flowAbortMode=LowerPriority
                  nodeName=Within Primary Melee Approach Range?
                  activeAbilityTagsToSkipTestingValue {key=None, defaultValue {gameplayTags []}}
                  distanceComparisons [{
                    addedOffset=0
                    bUseAddedOffset=false
                    bUseOverriddenValue=false
                    comparisonType=ThreeDimensions
                    operator=Less
                    overriddenValue=0
                    distanceDataTags {gameplayTags [{tagName=NPC.Ability.MaxApproachRange}]}
                   }]
                  gameplayAbilityTagContainer {key=None, defaultValue {gameplayTags [{tagName=NPC.Ability.Attack.Melee.Primary}]}}
                  targetValue {baseClass=&Actor, defaultValue=null, key=GoalActor}
                 }
                }
                Within Primary Melee Approach Range? {
                 class=FortBTDecorator_GameplayAbility_CompareDistance
                 properties {
                  bInverseCondition=false
                  flowAbortMode=LowerPriority
                  nodeName=Within Primary Melee Approach Range?
                  activeAbilityTagsToSkipTestingValue {key=None, defaultValue {gameplayTags []}}
                  distanceComparisons [{
                    addedOffset=0
                    bUseAddedOffset=false
                    bUseOverriddenValue=false
                    comparisonType=ThreeDimensions
                    operator=Greater
                    overriddenValue=0
                    distanceDataTags {gameplayTags [{tagName=NPC.Ability.MinApproachRange}]}
                   }]
                  gameplayAbilityTagContainer {key=None, defaultValue {gameplayTags [{tagName=NPC.Ability.Attack.Melee.Primary}]}}
                  targetValue {baseClass=&Actor, defaultValue=null, key=GoalActor}
                 }
                }
                Gameplay Tag Query {
                 class=BTDecorator_GameplayTagQuery
                 properties {
                  flowAbortMode=Both
                  nodeName=Gameplay Tag Query
                  actorForGameplayTagQuery {selectedKeyName=GoalActor}
                  gameplayTagQuery {
                   autoDescription=" NONE( Pawn.HidePlayerPawn )"
                   userDescription=Is Visible
                   queryTokenStream [0,1,3,1,0]
                   tagDictionary [{tagName=Pawn.HidePlayerPawn}]
                  }
                 }
                }
               ]
               properties {
                children [{
                  childComposite=&self:BTComposite_Selector_1
                  childTask=null
                  decoratorOps [
                   {number=2, operation=And}
                   {number=2, operation=Or}
                   {number=2, operation=And}
                   {number=0}
                   {number=1}
                   {number=2, operation=And}
                   {number=2}
                   {number=3}
                   {number=4}
                  ]
                  decorators [
                   &self:BTDecorator_CheckGameplayTagsOnActor_28
                   &self:TargetSlotsAvailable_C_3
                   &self:BTDecorator_CheckGameplayTagsOnActor_30
                   &self:BTDecorator_CheckGameplayTagsOnActor_29
                   &self:BTDecorator_CheckGameplayTagsOnActor_31
                  ]
                 }]
                services [
                 &self:FortBTService_ActivateAbility_4
                 &self:FortBTService_ActivateAbility_5
                ]
               }
               services [
                Activate Ability - Melee Primary {
                 class=FortBTService_ActivateAbility
                 properties {
                  bCallTickOnSearchStart=false
                  bPawnTargetsOnly=false
                  bRequireCanHitTargetWithAbility=true
                  bRestartTimerOnEachActivation=false
                  bShowEventDetails=false
                  bShowPropertyDetails=false
                  customDescription=""
                  interval=0.5
                  nodeName=Activate Ability - Melee Primary
                  randomDeviation=0
                  abilityTags {key=None, defaultValue {gameplayTags [{tagName=NPC.Ability.Attack.Melee.Primary}]}}
                  activateOnBecomeRelevant {defaultValue=false, key=None}
                  bCanActivateWhenMoving {defaultValue=true, key=None}
                  deactivateOnCeaseRelevant {defaultValue=false, key=None}
                  distanceChecks [{
                    addedOffset=0
                    bUseAddedOffset=false
                    bUseOverriddenValue=false
                    comparisonType=ThreeDimensions
                    operator=Less
                    overriddenValue=0
                    distanceDataTags {gameplayTags [{tagName=NPC.Ability.MaxActualRange}]}
                   }]
                  initialDelayRandomDeviation {defaultValue=0, key=None}
                  initialDelayTime {defaultValue=0, key=None}
                  prohibitedTargetTags {gameplayTags []}
                  targetActorKey {bNoneIsAllowedValue=false, selectedKeyName=GoalActor}
                  wantsReceiveActivatedAbilityEnded {defaultValue=false, key=None}
                 }
                }
                Activate Ability - Melee Lunge {
                 class=FortBTService_ActivateAbility
                 properties {
                  bCallTickOnSearchStart=false
                  bPawnTargetsOnly=false
                  bRequireCanHitTargetWithAbility=true
                  bRestartTimerOnEachActivation=false
                  bShowEventDetails=false
                  bShowPropertyDetails=false
                  customDescription=""
                  interval=0.5
                  nodeName=Activate Ability - Melee Lunge
                  randomDeviation=0
                  abilityTags {key=None, defaultValue {gameplayTags [{tagName=NPC.Ability.Attack.Melee.Lunge}]}}
                  activateOnBecomeRelevant {defaultValue=false, key=None}
                  bCanActivateWhenMoving {defaultValue=true, key=None}
                  deactivateOnCeaseRelevant {defaultValue=false, key=None}
                  distanceChecks [
                   {
                    addedOffset=0
                    bUseAddedOffset=false
                    bUseOverriddenValue=false
                    comparisonType=ThreeDimensions
                    operator=Less
                    overriddenValue=0
                    distanceDataTags {gameplayTags [{tagName=NPC.Ability.MaxActualRange}]}
                   }
                   {
                    addedOffset=0
                    bUseAddedOffset=false
                    bUseOverriddenValue=false
                    comparisonType=ThreeDimensions
                    operator=Greater
                    overriddenValue=0
                    distanceDataTags {gameplayTags [{tagName=NPC.Ability.MinPreferredRange}]}
                   }
                  ]
                  initialDelayRandomDeviation {defaultValue=0, key=None}
                  initialDelayTime {defaultValue=0, key=None}
                  prohibitedTargetTags {gameplayTags []}
                  targetActorKey {bNoneIsAllowedValue=false, selectedKeyName=GoalActor}
                  wantsReceiveActivatedAbilityEnded {defaultValue=false, key=None}
                 }
                }
               ]
              }]
             decorators [Has Melee Attack Ability? {
               class=FortBTDecorator_GameplayAbility_HasGameplayAbility
               properties {nodeName=Has Melee Attack Ability?, gameplayAbilityTagContainer {defaultValue {gameplayTags [{tagName=NPC.Ability.Attack.Melee}]}}}
              }]
             properties {
              children [{
                childComposite=&self:BTComposite_Selector_12
                childTask=null
                decorators [
                 &self:FortBTDecorator_GameplayAbility_HasGameplayAbility_40
                 &self:FortBTDecorator_GameplayAbility_HasGameplayAbility_41
                 &self:FortBTDecorator_GameplayAbility_CompareDistance_27
                 &self:FortBTDecorator_GameplayAbility_CompareDistance_30
                 &self:BTDecorator_GameplayTagQuery_26
                ]
               }]
              services []
             }
             services []
            }]
           decorators [
            Blackboard Based Condition {
             class=BTDecorator_Blackboard
             properties {
              arithmeticOperation=Equal
              basicOperation=NotSet
              flowAbortMode=LowerPriority
              intValue=0
              nodeName=Blackboard Based Condition
              notifyObserver=ValueChange
              stringValue=""
              blackboardKey {selectedKeyName=Leader}
             }
            }
            Blackboard Based Condition {
             class=BTDecorator_Blackboard
             properties {
              arithmeticOperation=Equal
              flowAbortMode=LowerPriority
              intValue=0
              nodeName=Blackboard Based Condition
              notifyObserver=ValueChange
              stringValue=""
              blackboardKey {selectedKeyName=Leader}
             }
            }
            AI Close to Leader? {
             class=FortBTDecorator_DistanceBetween
             properties {
              flowAbortMode=Both
              nodeName=AI Close to Leader?
              operator=LessOrEqual
              blackboardKeyB {selectedKeyName=Leader}
              specifiedDistance {defaultValue=4000}
             }
            }
            Leader close to AI's Goal? {
             class=FortBTDecorator_DistanceBetween
             properties {
              flowAbortMode=LowerPriority
              nodeName=Leader close to AI's Goal?
              operator=LessOrEqual
              blackboardKeyA {selectedKeyName=Leader}
              blackboardKeyB {selectedKeyName=GoalActor}
              specifiedDistance {defaultValue=3000}
             }
            }
           ]
           properties {
            children [{
              childComposite=&self:BTComposite_Selector_5
              childTask=null
              decorators [&self:FortBTDecorator_GameplayAbility_HasGameplayAbility_14]
             }]
            services []
           }
           services []
          }]
         decorators [NOT Intro or Idle Behavior? {
           class=BTDecorator_CheckGameplayTagsOnActor
           properties {
            bInverseCondition=true
            flowAbortMode=None
            nodeName=NOT Intro or Idle Behavior?
            tagsToMatch=Any
            actorToCheck {bNoneIsAllowedValue=false, selectedKeyName=SelfActor}
            gameplayTags {gameplayTags [{tagName=NPC.Behavior.Phase.Intro.BehaviorActive}, {tagName=NPC.Behavior.Phase.Idle}]}
           }
          }]
         properties {
          children [{
            childComposite=&self:BTComposite_Selector_24
            childTask=null
            decoratorOps [{number=2, operation=Or}, {number=0}, {number=3, operation=And}, {number=1}, {number=2}, {number=3}]
            decorators [
             &self:BTDecorator_Blackboard_15
             &self:BTDecorator_Blackboard_16
             &self:FortBTDecorator_DistanceBetween_6
             &self:FortBTDecorator_DistanceBetween_7
            ]
           }]
          services []
         }
         services []
        }
        Selector {
         class=BTComposite_Selector
         children [
          Selector {
           class=BTComposite_Selector
           children [
            GameplayAbility {
             class=FortBTTask_ExecuteGameplayAbility
             decorators [Can Activate Gameplay Ability {
               class=FortBTDecorator_GameplayAbility_CanActivate
               properties {
                flowAbortMode=None
                gameplayAbilityTagContainer {defaultValue {gameplayTags [{tagName=NPC.Ability.NonAttack.Reposition.ForTargetSlots.Melee}]}}
                targetValue {key=SpawnSpot}
               }
              }]
             properties {
              nodeName=GameplayAbility
              gameplayAbilityTagContainer {defaultValue {gameplayTags [{tagName=NPC.Ability.NonAttack.Reposition.ForTargetSlots.Melee}]}}
             }
            }
            Selector {
             class=BTComposite_Selector
             children [
              Rotate to face BB entry {
               class=BTTask_RotateToFaceBBEntry
               decorators [NOT Rotated To Attack Target {
                 class=FortBTDecorator_GameplayAbility_IsRotatedToAttackTarget
                 properties {
                  bInverseCondition=true
                  flowAbortMode=None
                  nodeName=NOT Rotated To Attack Target
                  activeAbilityTagsToSkipTestingValue {key=None, defaultValue {gameplayTags []}}
                  gameplayAbilityTagContainer {key=None, defaultValue {gameplayTags [{tagName=NPC.Ability.Attack.Melee.Primary}]}}
                  targetValue {baseClass=&Actor, defaultValue=null, key=GoalActor}
                 }
                }]
               properties {blackboardKey {selectedKeyName=GoalActor}, precision {defaultValue=10}}
              }
              Wait {class=BTTask_Wait, properties {randomDeviation {defaultValue=0.25}, waitTime {defaultValue=1}}}
             ]
             decorators [Is On Cooldown {
               class=FortBTDecorator_GameplayAbility_IsOnCooldown
               properties {
                bInverseCondition=false
                flowAbortMode=Self
                nodeName=Is On Cooldown
                activeAbilityTagsToSkipTestingValue {key=None, defaultValue {gameplayTags []}}
                gameplayAbilityTagContainer {key=None, defaultValue {gameplayTags [{tagName=NPC.Ability.NonAttack.Reposition.ForTargetSlots.Melee}]}}
                targetValue {baseClass=&Actor, defaultValue=null, key=SpawnSpot}
               }
              }]
             properties {
              children [
               {
                childComposite=null
                childTask=&self:BTTask_RotateToFaceBBEntry_2
                decorators [&self:FortBTDecorator_GameplayAbility_IsRotatedToAttackTarget_3]
               }
               {childComposite=null, childTask=&self:BTTask_Wait_9, decorators []}
              ]
              services []
             }
             services []
            }
           ]
           decorators [
            Does path exist {
             class=BTDecorator_DoesPathExist
             properties {
              bInverseCondition=false
              flowAbortMode=None
              nodeName=Does path exist
              blackboardKeyA {bNoneIsAllowedValue=false, selectedKeyName=SelfActor}
              blackboardKeyB {bNoneIsAllowedValue=false, selectedKeyName=GoalActor}
              filterClass {baseClass=&NavigationQueryFilter, defaultValue=null, key=None}
              pathQueryType {defaultValue=2, enumType=&EPathExistanceQueryType, key=None, nativeEnumTypeName=EPathExistanceQueryType}
             }
            }
            Within Primary Melee Approach Range? {
             class=FortBTDecorator_GameplayAbility_CompareDistance
             properties {
              bInverseCondition=false
              flowAbortMode=LowerPriority
              nodeName=Within Primary Melee Approach Range?
              activeAbilityTagsToSkipTestingValue {key=None, defaultValue {gameplayTags []}}
              distanceComparisons [{
                addedOffset=0
                bUseAddedOffset=false
                bUseOverriddenValue=false
                comparisonType=ThreeDimensions
                operator=Less
                overriddenValue=0
                distanceDataTags {gameplayTags [{tagName=NPC.Ability.MaxApproachRange}]}
               }]
              gameplayAbilityTagContainer {key=None, defaultValue {gameplayTags [{tagName=NPC.Ability.Attack.Melee.Primary}]}}
              targetValue {baseClass=&Actor, defaultValue=null, key=GoalActor}
             }
            }
           ]
           properties {
            children [
             {
              childComposite=null
              childTask=&self:FortBTTask_ExecuteGameplayAbility_1
              decorators [&self:FortBTDecorator_GameplayAbility_CanActivate_11]
             }
             {
              childComposite=&self:BTComposite_Selector_33
              childTask=null
              decorators [&self:FortBTDecorator_GameplayAbility_IsOnCooldown_8]
             }
            ]
            services [&self:FortBTService_ActivateAbility_3]
           }
           services [Activate Ability - Melee Primary {
             class=FortBTService_ActivateAbility
             properties {
              bCallTickOnSearchStart=false
              bPawnTargetsOnly=false
              bRequireCanHitTargetWithAbility=true
              bRestartTimerOnEachActivation=false
              bShowEventDetails=false
              bShowPropertyDetails=false
              customDescription=""
              interval=0.5
              nodeName=Activate Ability - Melee Primary
              randomDeviation=0
              abilityTags {key=None, defaultValue {gameplayTags [{tagName=NPC.Ability.Attack.Melee.Primary}]}}
              activateOnBecomeRelevant {defaultValue=false, key=None}
              bCanActivateWhenMoving {defaultValue=true, key=None}
              deactivateOnCeaseRelevant {defaultValue=false, key=None}
              distanceChecks [{
                addedOffset=0
                bUseAddedOffset=false
                bUseOverriddenValue=false
                comparisonType=ThreeDimensions
                operator=Less
                overriddenValue=0
                distanceDataTags {gameplayTags [{tagName=NPC.Ability.MaxActualRange}]}
               }]
              initialDelayRandomDeviation {defaultValue=0, key=None}
              initialDelayTime {defaultValue=0, key=None}
              prohibitedTargetTags {gameplayTags []}
              targetActorKey {bNoneIsAllowedValue=false, selectedKeyName=GoalActor}
              wantsReceiveActivatedAbilityEnded {defaultValue=false, key=None}
             }
            }]
          }
          FortMoveTo {
           class=FortBTTask_GameMoveTo
           decorators [
            Does path exist - Can Destroy Buildings {
             class=BTDecorator_DoesPathExist
             properties {
              bInverseCondition=false
              flowAbortMode=None
              nodeName=Does path exist - Can Destroy Buildings
              blackboardKeyA {bNoneIsAllowedValue=false, selectedKeyName=SelfActor}
              blackboardKeyB {bNoneIsAllowedValue=false, selectedKeyName=GoalActor}
              filterClass {baseClass=&NavigationQueryFilter, defaultValue=null, key=None}
              pathQueryType {defaultValue=2, enumType=&EPathExistanceQueryType, key=None, nativeEnumTypeName=EPathExistanceQueryType}
             }
            }
            NOT Within Primary Melee Approach Range? {
             class=FortBTDecorator_GameplayAbility_CompareDistance
             properties {
              bInverseCondition=false
              flowAbortMode=LowerPriority
              nodeName=NOT Within Primary Melee Approach Range?
              activeAbilityTagsToSkipTestingValue {key=None, defaultValue {gameplayTags []}}
              distanceComparisons [{
                addedOffset=0
                bUseAddedOffset=false
                bUseOverriddenValue=false
                comparisonType=ThreeDimensions
                operator=Greater
                overriddenValue=0
                distanceDataTags {gameplayTags [{tagName=NPC.Ability.MaxApproachRange}]}
               }]
              gameplayAbilityTagContainer {key=None, defaultValue {gameplayTags [{tagName=NPC.Ability.Attack.Melee.Primary}]}}
              targetValue {baseClass=&Actor, defaultValue=null, key=GoalActor}
             }
            }
           ]
           properties {
            bDeimosFlavor=false
            bFinishMoveOnOverlap=true
            bStopAtGoal=false
            nodeName=FortMoveTo
            acceptableRadius {defaultValue=30}
            bReachTestIncludesAgentRadius {defaultValue=true}
            bReachTestIncludesGoalRadius {defaultValue=true}
            blackboardKey {selectedKeyName=GoalActor}
            filterClass {defaultValue=&NavFilter_NPC_AnyTraversable_C}
            navFilterTag {tagName=NavFilter.NPC.DefaultMove}
           }
          }
          GameplayAbility {
           class=FortBTTask_ExecuteGameplayAbility
           decorators [
            Has Gameplay Ability {
             class=FortBTDecorator_GameplayAbility_HasGameplayAbility
             properties {gameplayAbilityTagContainer {defaultValue {gameplayTags [{tagName=NPC.Ability.NonAttack.PlayMontage.Wander.NearGoal}]}}}
            }
            Has Phase Wander Animation Pending {
             class=BTDecorator_GameplayTagQuery
             properties {
              flowAbortMode=None
              nodeName=Has Phase Wander Animation Pending
              gameplayTagQuery {
               autoDescription=" ANY( NPC.Behavior.Phase.Wander.NearGoal.Animation.Pending )"
               queryTokenStream [0,1,1,1,0]
               tagDictionary [{tagName=NPC.Behavior.Phase.Wander.NearGoal.Animation.Pending}]
              }
             }
            }
            Can Activate Gameplay Ability {
             class=FortBTDecorator_GameplayAbility_CanActivate
             properties {
              flowAbortMode=None
              gameplayAbilityTagContainer {defaultValue {gameplayTags [{tagName=NPC.Ability.NonAttack.PlayMontage.Wander.NearGoal}]}}
              targetValue {key=SpawnSpot}
             }
            }
           ]
           properties {
            nodeName=GameplayAbility
            gameplayAbilityTagContainer {defaultValue {gameplayTags [{tagName=NPC.Ability.NonAttack.PlayMontage.Wander.NearGoal}]}}
           }
          }
          GameplayAbility {
           class=FortBTTask_ExecuteGameplayAbility
           decorators [
            Has Gameplay Ability {
             class=FortBTDecorator_GameplayAbility_HasGameplayAbility
             properties {gameplayAbilityTagContainer {defaultValue {gameplayTags [{tagName=NPC.Ability.NonAttack.Wander.NearGoal}]}}}
            }
            Can Activate Gameplay Ability {
             class=FortBTDecorator_GameplayAbility_CanActivate
             properties {
              flowAbortMode=LowerPriority
              gameplayAbilityTagContainer {defaultValue {gameplayTags [{tagName=NPC.Ability.NonAttack.Wander.NearGoal}]}}
              targetValue {key=SpawnSpot}
             }
            }
           ]
           properties {nodeName=GameplayAbility, gameplayAbilityTagContainer {defaultValue {gameplayTags [{tagName=NPC.Ability.NonAttack.Wander.NearGoal}]}}}
          }
         ]
         decorators [Has ANY Attack Ability? {
           class=FortBTDecorator_GameplayAbility_HasGameplayAbility
           properties {nodeName=Has ANY Attack Ability?, gameplayAbilityTagContainer {defaultValue {gameplayTags [{tagName=NPC.Ability.Attack}]}}}
          }]
         properties {
          children [
           {
            childComposite=&self:BTComposite_Selector_35
            childTask=null
            decorators [
             &self:BTDecorator_DoesPathExist_2
             &self:FortBTDecorator_GameplayAbility_CompareDistance_11
            ]
           }
           {
            childComposite=null
            childTask=&self:FortBTTask_GameMoveTo_15
            decorators [
             &self:BTDecorator_DoesPathExist_4
             &self:FortBTDecorator_GameplayAbility_CompareDistance_26
            ]
           }
           {
            childComposite=null
            childTask=&self:FortBTTask_ExecuteGameplayAbility_21
            decorators [
             &self:FortBTDecorator_GameplayAbility_HasGameplayAbility_27
             &self:BTDecorator_GameplayTagQuery_18
             &self:FortBTDecorator_GameplayAbility_CanActivate_18
            ]
           }
           {
            childComposite=null
            childTask=&self:FortBTTask_ExecuteGameplayAbility_11
            decorators [
             &self:FortBTDecorator_GameplayAbility_HasGameplayAbility_7
             &self:FortBTDecorator_GameplayAbility_CanActivate_8
            ]
           }
          ]
          services []
         }
         services []
        }
       ]
       decorators [Is Goal Pawn {
         class=FortBTDecorator_IsGoalPawn
         properties {bInverseCondition=false, flowAbortMode=Both, nodeName=Is Goal Pawn, blackboardKey {bNoneIsAllowedValue=false, selectedKeyName=GoalActor}}
        }]
       properties {
        children [
         {
          childComposite=&self:BTComposite_Selector_21
          childTask=null
          decorators [&self:BTDecorator_CheckGameplayTagsOnActor_4]
         }
         {
          childComposite=&self:BTComposite_Selector_30
          childTask=null
          decorators [&self:FortBTDecorator_GameplayAbility_HasGameplayAbility_13]
         }
        ]
        services []
       }
       services []
      }
      Selector {
       class=BTComposite_Selector
       children [
        Selector {
         class=BTComposite_Selector
         children [Selector {
           class=BTComposite_Selector
           children [
            Selector {
             class=BTComposite_Selector
             children [GameplayAbility {
               class=FortBTTask_ExecuteGameplayAbility
               decorators [
                Has Gameplay Ability {
                 class=FortBTDecorator_GameplayAbility_HasGameplayAbility
                 properties {gameplayAbilityTagContainer {defaultValue {gameplayTags [{tagName=NPC.Ability.NonAttack.Defense}]}}}
                }
                Can Activate Gameplay Ability {
                 class=FortBTDecorator_GameplayAbility_CanActivate
                 properties {
                  flowAbortMode=LowerPriority
                  gameplayAbilityTagContainer {defaultValue {gameplayTags [{tagName=NPC.Ability.NonAttack.Defense}]}}
                  targetValue {key=GoalActor}
                 }
                }
                Compare Gameplay Ability Distances {
                 class=FortBTDecorator_GameplayAbility_CompareDistance
                 properties {
                  bInverseCondition=false
                  flowAbortMode=LowerPriority
                  nodeName=Compare Gameplay Ability Distances
                  activeAbilityTagsToSkipTestingValue {key=None, defaultValue {gameplayTags []}}
                  distanceComparisons [{
                    addedOffset=0
                    bUseAddedOffset=false
                    bUseOverriddenValue=false
                    comparisonType=ThreeDimensions
                    operator=Less
                    overriddenValue=0
                    distanceDataTags {gameplayTags [{tagName=NPC.Ability.MaxApproachRange}]}
                   }]
                  gameplayAbilityTagContainer {key=None, defaultValue {gameplayTags [{tagName=NPC.Ability.NonAttack.Defense}]}}
                  targetValue {baseClass=&Actor, defaultValue=null, key=GoalActor}
                 }
                }
                Gameplay Tag Query {
                 class=BTDecorator_GameplayTagQuery
                 properties {
                  flowAbortMode=LowerPriority
                  nodeName=Gameplay Tag Query
                  gameplayTagQuery {
                   autoDescription=" ANY( Gameplay.Status.RecentlyDamaged )"
                   queryTokenStream [0,1,1,1,0]
                   tagDictionary [{tagName=Gameplay.Status.RecentlyDamaged}]
                  }
                 }
                }
               ]
               properties {nodeName=GameplayAbility, gameplayAbilityTagContainer {defaultValue {gameplayTags [{tagName=NPC.Ability.NonAttack.Defense}]}}}
              }]
             decorators [Blackboard Based Condition {
               class=BTDecorator_Blackboard
               properties {
                arithmeticOperation=Equal
                intValue=0
                nodeName=Blackboard Based Condition
                notifyObserver=ValueChange
                stringValue=""
                blackboardKey {selectedKeyName=Leader}
               }
              }]
             properties {
              children [{
                childComposite=null
                childTask=&self:FortBTTask_ExecuteGameplayAbility_37
                decoratorOps [{number=3, operation=And}, {number=0}, {number=1}, {number=2, operation=Or}, {number=2}, {number=3}]
                decorators [
                 &self:FortBTDecorator_GameplayAbility_HasGameplayAbility_39
                 &self:FortBTDecorator_GameplayAbility_CanActivate_7
                 &self:FortBTDecorator_GameplayAbility_CompareDistance_25
                 &self:BTDecorator_GameplayTagQuery_23
                ]
               }]
              services []
             }
             services []
            }
            Selector {
             class=BTComposite_Selector
             children [GameplayAbility {
               class=FortBTTask_ExecuteGameplayAbility
               decorators [
                Has Gameplay Ability {
                 class=FortBTDecorator_GameplayAbility_HasGameplayAbility
                 properties {gameplayAbilityTagContainer {defaultValue {gameplayTags [{tagName=NPC.Ability.NonAttack.Avoid.Enemy.Panic}]}}}
                }
                Can Activate Gameplay Ability {
                 class=FortBTDecorator_GameplayAbility_CanActivate
                 properties {
                  flowAbortMode=LowerPriority
                  gameplayAbilityTagContainer {defaultValue {gameplayTags [{tagName=NPC.Ability.NonAttack.Avoid.Enemy.Panic}]}}
                  targetValue {key=GoalActor}
                 }
                }
                Compare Gameplay Ability Distances {
                 class=FortBTDecorator_GameplayAbility_CompareDistance
                 properties {
                  bInverseCondition=false
                  flowAbortMode=LowerPriority
                  nodeName=Compare Gameplay Ability Distances
                  activeAbilityTagsToSkipTestingValue {key=None, defaultValue {gameplayTags []}}
                  distanceComparisons [{
                    addedOffset=0
                    bUseAddedOffset=false
                    bUseOverriddenValue=false
                    comparisonType=ThreeDimensions
                    operator=Less
                    overriddenValue=0
                    distanceDataTags {gameplayTags [{tagName=NPC.Ability.MaxApproachRange}]}
                   }]
                  gameplayAbilityTagContainer {key=None, defaultValue {gameplayTags [{tagName=NPC.Ability.NonAttack.Avoid.Enemy.Charge}]}}
                  targetValue {baseClass=&Actor, defaultValue=null, key=GoalActor}
                 }
                }
                Gameplay Tag Query {
                 class=BTDecorator_GameplayTagQuery
                 properties {
                  flowAbortMode=LowerPriority
                  nodeName=Gameplay Tag Query
                  gameplayTagQuery {
                   autoDescription=" ANY( Gameplay.Status.RecentlyDamaged )"
                   queryTokenStream [0,1,1,1,0]
                   tagDictionary [{tagName=Gameplay.Status.RecentlyDamaged}]
                  }
                 }
                }
               ]
               properties {nodeName=GameplayAbility, gameplayAbilityTagContainer {defaultValue {gameplayTags [{tagName=NPC.Ability.NonAttack.Avoid.Enemy.Panic}]}}}
              }]
             decorators [Blackboard Based Condition {
               class=BTDecorator_Blackboard
               properties {
                arithmeticOperation=Equal
                basicOperation=NotSet
                intValue=0
                nodeName=Blackboard Based Condition
                notifyObserver=ValueChange
                stringValue=""
                blackboardKey {selectedKeyName=Leader}
               }
              }]
             properties {
              children [{
                childComposite=null
                childTask=&self:FortBTTask_ExecuteGameplayAbility_9
                decoratorOps [{number=3, operation=And}, {number=0}, {number=1}, {number=2, operation=Or}, {number=2}, {number=3}]
                decorators [
                 &self:FortBTDecorator_GameplayAbility_HasGameplayAbility_1
                 &self:FortBTDecorator_GameplayAbility_CanActivate_20
                 &self:FortBTDecorator_GameplayAbility_CompareDistance_14
                 &self:BTDecorator_GameplayTagQuery_22
                ]
               }]
              services []
             }
             services []
            }
            Selector {
             class=BTComposite_Selector
             children [Sequence {
               class=BTComposite_Sequence
               children [
                Wait {class=BTTask_Wait, properties {waitTime {defaultValue=0.5}}}
                GameplayAbility {
                 class=FortBTTask_ExecuteGameplayAbility
                 decorators []
                 properties {
                  nodeName=GameplayAbility
                  gameplayAbilityTagContainer {defaultValue {gameplayTags [{tagName=NPC.Ability.NonAttack.Frustration.DueTo.Movement}]}}
                 }
                }
               ]
               decorators []
               properties {children [
                 {childTask=&self:BTTask_Wait_13}
                 {childTask=&self:FortBTTask_ExecuteGameplayAbility_18}
                ]}
              }]
             decorators [
              Is NOT Falling? {
               class=BTDecorator_GameplayTagQuery
               properties {
                flowAbortMode=LowerPriority
                nodeName=Is NOT Falling?
                gameplayTagQuery {autoDescription=" NONE( MovementMode.Falling )", queryTokenStream [0,1,3,1,0], tagDictionary [{tagName=MovementMode.Falling}]}
               }
              }
              Has Gameplay Ability {
               class=FortBTDecorator_GameplayAbility_HasGameplayAbility
               properties {gameplayAbilityTagContainer {defaultValue {gameplayTags [{tagName=NPC.Ability.NonAttack.Frustration.DueTo.Movement}]}}}
              }
              NPC.Status.Frustration.DueTo.Movement {
               class=BTDecorator_CheckGameplayTagsOnActor
               properties {
                bInverseCondition=false
                flowAbortMode=None
                nodeName=NPC.Status.Frustration.DueTo.Movement
                tagsToMatch=Any
                actorToCheck {bNoneIsAllowedValue=false, selectedKeyName=SelfActor}
                gameplayTags {gameplayTags [{tagName=NPC.Status.Frustration.DueTo.Movement}]}
               }
              }
             ]
             properties {children [{childComposite=&self:BTComposite_Sequence_10, childTask=null, decorators []}], services []}
             services []
            }
            GameplayAbility {
             class=FortBTTask_ExecuteGameplayAbility
             decorators [
              Has Gameplay Ability {
               class=FortBTDecorator_GameplayAbility_HasGameplayAbility
               properties {gameplayAbilityTagContainer {defaultValue {gameplayTags [{tagName=NPC.Ability.NonAttack.Avoid.Enemy.Primary}]}}}
              }
              Can Activate Gameplay Ability {
               class=FortBTDecorator_GameplayAbility_CanActivate
               properties {
                flowAbortMode=LowerPriority
                gameplayAbilityTagContainer {defaultValue {gameplayTags [{tagName=NPC.Ability.NonAttack.Avoid.Enemy.Primary}]}}
                targetValue {key=GoalActor}
               }
              }
              Compare Gameplay Ability Distances {
               class=FortBTDecorator_GameplayAbility_CompareDistance
               properties {
                bInverseCondition=false
                flowAbortMode=LowerPriority
                nodeName=Compare Gameplay Ability Distances
                activeAbilityTagsToSkipTestingValue {key=None, defaultValue {gameplayTags []}}
                distanceComparisons [{
                  addedOffset=0
                  bUseAddedOffset=false
                  bUseOverriddenValue=false
                  comparisonType=ThreeDimensions
                  operator=Less
                  overriddenValue=0
                  distanceDataTags {gameplayTags [{tagName=NPC.Ability.MaxApproachRange}]}
                 }]
                gameplayAbilityTagContainer {key=None, defaultValue {gameplayTags [{tagName=NPC.Ability.NonAttack.Avoid.Enemy.Primary}]}}
                targetValue {baseClass=&Actor, defaultValue=null, key=GoalActor}
               }
              }
              Gameplay Tag Query {
               class=BTDecorator_GameplayTagQuery
               properties {
                flowAbortMode=LowerPriority
                nodeName=Gameplay Tag Query
                gameplayTagQuery {
                 autoDescription=" ANY( Gameplay.Status.RecentlyDamaged, NPC.Behavior.Phase.Avoid.Enemy.Active )"
                 queryTokenStream [0,1,1,2,0,1]
                 tagDictionary [{tagName=Gameplay.Status.RecentlyDamaged}, {tagName=NPC.Behavior.Phase.Avoid.Enemy.Active}]
                }
               }
              }
             ]
             properties {nodeName=GameplayAbility, gameplayAbilityTagContainer {defaultValue {gameplayTags [{tagName=NPC.Ability.NonAttack.Avoid.Enemy.Primary}]}}}
            }
            Selector {
             class=BTComposite_Selector
             children [Wait {class=BTTask_Wait, properties {randomDeviation {defaultValue=0.25}, waitTime {defaultValue=1}}}]
             decorators [Is On Cooldown {
               class=FortBTDecorator_GameplayAbility_IsOnCooldown
               properties {
                bInverseCondition=false
                flowAbortMode=Self
                nodeName=Is On Cooldown
                activeAbilityTagsToSkipTestingValue {key=None, defaultValue {gameplayTags []}}
                gameplayAbilityTagContainer {key=None, defaultValue {gameplayTags [{tagName=NPC.Ability.NonAttack.Avoid.Enemy.Primary}]}}
                targetValue {baseClass=&Actor, defaultValue=null, key=SpawnSpot}
               }
              }]
             properties {children [{childComposite=null, childTask=&self:BTTask_Wait_4, decorators []}], services []}
             services []
            }
           ]
           decorators [NOT Intro or Idle Behavior? {
             class=BTDecorator_CheckGameplayTagsOnActor
             properties {
              bInverseCondition=true
              flowAbortMode=None
              nodeName=NOT Intro or Idle Behavior?
              tagsToMatch=Any
              actorToCheck {bNoneIsAllowedValue=false, selectedKeyName=SelfActor}
              gameplayTags {gameplayTags [{tagName=NPC.Behavior.Phase.Intro.BehaviorActive}, {tagName=NPC.Behavior.Phase.Idle}]}
             }
            }]
           properties {
            children [
             {
              childComposite=&self:BTComposite_Selector_6
              childTask=null
              decorators [&self:BTDecorator_Blackboard_4]
             }
             {
              childComposite=&self:BTComposite_Selector_46
              childTask=null
              decorators [&self:BTDecorator_Blackboard_6]
             }
             {
              childComposite=&self:BTComposite_Selector_31
              childTask=null
              decorators [
               &self:BTDecorator_GameplayTagQuery_12
               &self:FortBTDecorator_GameplayAbility_HasGameplayAbility_22
               &self:BTDecorator_CheckGameplayTagsOnActor_17
              ]
             }
             {
              childComposite=null
              childTask=&self:FortBTTask_ExecuteGameplayAbility_8
              decoratorOps [{number=3, operation=And}, {number=0}, {number=1}, {number=2, operation=Or}, {number=2}, {number=3}]
              decorators [
               &self:FortBTDecorator_GameplayAbility_HasGameplayAbility_3
               &self:FortBTDecorator_GameplayAbility_CanActivate_2
               &self:FortBTDecorator_GameplayAbility_CompareDistance_12
               &self:BTDecorator_GameplayTagQuery_21
              ]
             }
             {
              childComposite=&self:BTComposite_Selector_37
              childTask=null
              decorators [&self:FortBTDecorator_GameplayAbility_IsOnCooldown_5]
             }
            ]
            services [&self:FortBTService_SendGameplayEvents_C_3]
           }
           services [Send Gameplay Events to Blackboard Key - Avoid Enemy Phase {
             class=FortBTService_SendGameplayEvents_C
             properties {
              activatingPawn=null
              activatingPawnAbilitySystemComponent=null
              bCallTickOnSearchStart=false
              bRestartTimerOnEachActivation=false
              bShowEventDetails=false
              bShowPropertyDetails=true
              customDescription=""
              goalActor=null
              interval=0.5
              nodeName=Send Gameplay Events to Blackboard Key - Avoid Enemy Phase
              ownerActor=null
              randomDeviation=0.1
              targetAbilitySystemComponent=null
              useEventTickLogic=false
              activationEventKeysTags []
              activationEventValuesBlackboards []
              activationEvents {"(TagName=\"NPC.Behavior.Phase.Avoid.Enemy.Apply\")" {bNoneIsAllowedValue=false, selectedKeyName=SelfActor}}
              deactivationEventKeysTags []
              deactivationEvents {}
              deactivationEventsValuesBlackboards []
             }
            }]
          }]
         decorators [Is Goal Pawn {
           class=FortBTDecorator_IsGoalPawn
           properties {bInverseCondition=false, flowAbortMode=Both, nodeName=Is Goal Pawn, blackboardKey {bNoneIsAllowedValue=false, selectedKeyName=GoalActor}}
          }]
         properties {
          children [{
            childComposite=&self:BTComposite_Selector_13
            childTask=null
            decorators [&self:BTDecorator_CheckGameplayTagsOnActor_7]
           }]
          services []
         }
         services []
        }
        Selector {
         class=BTComposite_Selector
         children [Sequence {
           class=BTComposite_Sequence
           children [
            Wait {class=BTTask_Wait, properties {waitTime {defaultValue=0.5}}}
            GameplayAbility {
             class=FortBTTask_ExecuteGameplayAbility
             decorators []
             properties {
              nodeName=GameplayAbility
              gameplayAbilityTagContainer {defaultValue {gameplayTags [{tagName=NPC.Ability.NonAttack.Frustration.DueTo.Movement}]}}
             }
            }
           ]
           decorators []
           properties {children [
             {childTask=&self:BTTask_Wait_11}
             {childTask=&self:FortBTTask_ExecuteGameplayAbility_16}
            ]}
          }]
         decorators [
          Is NOT Falling? {
           class=BTDecorator_GameplayTagQuery
           properties {
            flowAbortMode=LowerPriority
            nodeName=Is NOT Falling?
            gameplayTagQuery {autoDescription=" NONE( MovementMode.Falling )", queryTokenStream [0,1,3,1,0], tagDictionary [{tagName=MovementMode.Falling}]}
           }
          }
          Has Gameplay Ability {
           class=FortBTDecorator_GameplayAbility_HasGameplayAbility
           properties {gameplayAbilityTagContainer {defaultValue {gameplayTags [{tagName=NPC.Ability.NonAttack.Frustration.DueTo.Movement}]}}}
          }
          NPC.Status.Frustration.DueTo.Movement {
           class=BTDecorator_CheckGameplayTagsOnActor
           properties {
            bInverseCondition=false
            flowAbortMode=None
            nodeName=NPC.Status.Frustration.DueTo.Movement
            tagsToMatch=Any
            actorToCheck {bNoneIsAllowedValue=false, selectedKeyName=SelfActor}
            gameplayTags {gameplayTags [{tagName=NPC.Status.Frustration.DueTo.Movement}]}
           }
          }
         ]
         properties {children [{childComposite=&self:BTComposite_Sequence_3, childTask=null, decorators []}], services []}
         services []
        }
        Selector {
         class=BTComposite_Selector
         children [
          GameplayAbility {
           class=FortBTTask_ExecuteGameplayAbility
           decorators [MaxActualRange {
             class=FortBTDecorator_GameplayAbility_CompareDistance
             properties {
              bInverseCondition=false
              flowAbortMode=None
              nodeName=MaxActualRange
              activeAbilityTagsToSkipTestingValue {key=None, defaultValue {gameplayTags []}}
              distanceComparisons [{
                addedOffset=0
                bUseAddedOffset=false
                bUseOverriddenValue=false
                comparisonType=ThreeDimensions
                operator=Less
                overriddenValue=0
                distanceDataTags {gameplayTags [{tagName=NPC.Ability.MaxActualRange}]}
               }]
              gameplayAbilityTagContainer {key=None, defaultValue {gameplayTags [{tagName=NPC.Ability.NonAttack.GrabPickup}]}}
              targetValue {baseClass=&Actor, defaultValue=null, key=GoalActor}
             }
            }]
           properties {nodeName=GameplayAbility, gameplayAbilityTagContainer {defaultValue {gameplayTags [{tagName=NPC.Ability.NonAttack.GrabPickup}]}}}
          }
          Selector {
           class=BTComposite_Selector
           children [
            GameplayAbility {
             class=FortBTTask_ExecuteGameplayAbility
             decorators [Does path exist {
               class=BTDecorator_DoesPathExist
               properties {
                bInverseCondition=false
                flowAbortMode=None
                nodeName=Does path exist
                blackboardKeyA {bNoneIsAllowedValue=false, selectedKeyName=SelfActor}
                blackboardKeyB {bNoneIsAllowedValue=false, selectedKeyName=GoalActor}
                filterClass {baseClass=&NavigationQueryFilter, defaultValue=null, key=None}
                pathQueryType {defaultValue=2, enumType=&EPathExistanceQueryType, key=None, nativeEnumTypeName=EPathExistanceQueryType}
               }
              }]
             properties {nodeName=GameplayAbility, gameplayAbilityTagContainer {defaultValue {gameplayTags [{tagName=NPC.Ability.NonAttack.Move.ToPickup}]}}}
            }
            Set Frustration Discouragement {
             class=FortBTTask_SetFrustrationDiscouragement
             children []
             decorators []
             properties {bIgnoreRestartSelf=false, discouragementDuration=10, nodeName=Set Frustration Discouragement}
             services []
            }
           ]
           decorators [
            Has Move To Pickup Ability {
             class=FortBTDecorator_GameplayAbility_HasGameplayAbility
             properties {
              nodeName=Has Move To Pickup Ability
              gameplayAbilityTagContainer {defaultValue {gameplayTags [{tagName=NPC.Ability.NonAttack.Move.ToPickup}]}}
             }
            }
            Can Activate Gameplay Ability - Move To Pickup {
             class=FortBTDecorator_GameplayAbility_CanActivate
             properties {
              flowAbortMode=LowerPriority
              nodeName=Can Activate Gameplay Ability - Move To Pickup
              gameplayAbilityTagContainer {defaultValue {gameplayTags [{tagName=NPC.Ability.NonAttack.Move.ToPickup}]}}
              targetValue {key=GoalActor}
             }
            }
           ]
           properties {
            children [
             {
              childComposite=null
              childTask=&self:FortBTTask_ExecuteGameplayAbility_7
              decorators [&self:BTDecorator_DoesPathExist_3]
             }
             {childComposite=null, childTask=&self:FortBTTask_SetFrustrationDiscouragement_1, decorators []}
            ]
            services []
           }
           services []
          }
         ]
         decorators [
          Has GrabPickup Ability {
           class=FortBTDecorator_GameplayAbility_HasGameplayAbility
           properties {nodeName=Has GrabPickup Ability, gameplayAbilityTagContainer {defaultValue {gameplayTags [{tagName=NPC.Ability.NonAttack.GrabPickup}]}}}
          }
          Can Activate Gameplay Ability - Grab Pickup {
           class=FortBTDecorator_GameplayAbility_CanActivate
           properties {
            flowAbortMode=LowerPriority
            nodeName=Can Activate Gameplay Ability - Grab Pickup
            gameplayAbilityTagContainer {defaultValue {gameplayTags [{tagName=NPC.Ability.NonAttack.GrabPickup}]}}
            targetValue {key=GoalActor}
           }
          }
          HACK - Is Goal NOT Pawn - Observers below are LAZY {
           class=FortBTDecorator_IsGoalPawn
           properties {
            bInverseCondition=true
            flowAbortMode=Both
            nodeName=HACK - Is Goal NOT Pawn - Observers below are LAZY
            blackboardKey {bNoneIsAllowedValue=false, selectedKeyName=GoalActor}
           }
          }
          Is BlackBoard value of given Class {
           class=BTDecorator_IsBBEntryOfClass
           properties {
            bInverseCondition=false
            flowAbortMode=Both
            nodeName=Is BlackBoard value of given Class
            blackboardKey {bNoneIsAllowedValue=false, selectedKeyName=GoalActor}
            testClass {baseClass=&Object, defaultValue=&FortPickup, key=None}
           }
          }
          Is BlackBoard value of given Class {
           class=BTDecorator_IsBBEntryOfClass
           properties {
            bInverseCondition=false
            flowAbortMode=Both
            nodeName=Is BlackBoard value of given Class
            blackboardKey {bNoneIsAllowedValue=false, selectedKeyName=GoalActor}
            testClass {baseClass=&Object, defaultValue=&FortProjectileBase, key=None}
           }
          }
          Has GoalActor? To Abort Self for goals likely to disappear {
           class=BTDecorator_Blackboard
           properties {
            arithmeticOperation=Equal
            intValue=0
            nodeName=Has GoalActor? To Abort Self for goals likely to disappear
            notifyObserver=ValueChange
            stringValue=""
            blackboardKey {selectedKeyName=GoalActor}
           }
          }
         ]
         properties {
          children [
           {
            childComposite=null
            childTask=&self:FortBTTask_ExecuteGameplayAbility_2
            decorators [&self:FortBTDecorator_GameplayAbility_CompareDistance_5]
           }
           {
            childComposite=&self:BTComposite_Selector_48
            childTask=null
            decorators [
             &self:FortBTDecorator_GameplayAbility_HasGameplayAbility_12
             &self:FortBTDecorator_GameplayAbility_CanActivate_3
            ]
           }
          ]
          services []
         }
         services []
        }
        Selector {
         class=BTComposite_Selector
         children [
          GameplayAbility {
           class=FortBTTask_ExecuteGameplayAbility
           decorators [Can Activate Gameplay Ability {
             class=FortBTDecorator_GameplayAbility_CanActivate
             properties {
              flowAbortMode=None
              gameplayAbilityTagContainer {defaultValue {gameplayTags [{tagName=NPC.Ability.NonAttack.Move.ToLand}]}}
              targetValue {key=SpawnSpot}
             }
            }]
           properties {nodeName=GameplayAbility, gameplayAbilityTagContainer {defaultValue {gameplayTags [{tagName=NPC.Ability.NonAttack.Move.ToLand}]}}}
          }
          Wait {
           class=BTTask_Wait
           decorators [Is On Cooldown {
             class=FortBTDecorator_GameplayAbility_IsOnCooldown
             properties {
              bInverseCondition=false
              flowAbortMode=None
              nodeName=Is On Cooldown
              activeAbilityTagsToSkipTestingValue {key=None, defaultValue {gameplayTags []}}
              gameplayAbilityTagContainer {key=None, defaultValue {gameplayTags [{tagName=NPC.Ability.NonAttack.Move.ToLand}]}}
              targetValue {baseClass=&Actor, defaultValue=null, key=SpawnSpot}
             }
            }]
           properties {waitTime {defaultValue=1}}
          }
         ]
         decorators [
          Has Gameplay Ability {
           class=FortBTDecorator_GameplayAbility_HasGameplayAbility
           properties {gameplayAbilityTagContainer {defaultValue {gameplayTags [{tagName=NPC.Ability.NonAttack.Move.ToLand}]}}}
          }
          Is Surface Swimming {
           class=BTDecorator_GameplayTagQuery
           properties {
            flowAbortMode=LowerPriority
            nodeName=Is Surface Swimming
            gameplayTagQuery {autoDescription=" ANY( MovementMode.SurfaceSwimming )", queryTokenStream [0,1,1,1,0], tagDictionary [{tagName=MovementMode.SurfaceSwimming}]}
           }
          }
          Has no Leader {
           class=BTDecorator_Blackboard
           properties {
            arithmeticOperation=Equal
            basicOperation=NotSet
            intValue=0
            nodeName=Has no Leader
            notifyObserver=ValueChange
            stringValue=""
            blackboardKey {selectedKeyName=Leader}
           }
          }
         ]
         properties {
          children [
           {
            childComposite=null
            childTask=&self:FortBTTask_ExecuteGameplayAbility_24
            decorators [&self:FortBTDecorator_GameplayAbility_CanActivate_23]
           }
           {
            childComposite=null
            childTask=&self:BTTask_Wait_19
            decorators [&self:FortBTDecorator_GameplayAbility_IsOnCooldown_10]
           }
          ]
          services []
         }
         services []
        }
        Selector {
         class=BTComposite_Selector
         children [
          GameplayAbility {
           class=FortBTTask_ExecuteGameplayAbility
           decorators [Can Activate Gameplay Ability {
             class=FortBTDecorator_GameplayAbility_CanActivate
             properties {
              flowAbortMode=None
              gameplayAbilityTagContainer {defaultValue {gameplayTags [{tagName=NPC.Ability.NonAttack.Avoid.Enemy.Wander}]}}
              targetValue {key=SpawnSpot}
             }
            }]
           properties {nodeName=GameplayAbility, gameplayAbilityTagContainer {defaultValue {gameplayTags [{tagName=NPC.Ability.NonAttack.Avoid.Enemy.Wander}]}}}
          }
          Wait {
           class=BTTask_Wait
           decorators [Is On Cooldown {
             class=FortBTDecorator_GameplayAbility_IsOnCooldown
             properties {
              bInverseCondition=false
              flowAbortMode=None
              nodeName=Is On Cooldown
              activeAbilityTagsToSkipTestingValue {key=None, defaultValue {gameplayTags []}}
              gameplayAbilityTagContainer {key=None, defaultValue {gameplayTags [{tagName=NPC.Ability.NonAttack.Avoid.Enemy.Wander}]}}
              targetValue {baseClass=&Actor, defaultValue=null, key=SpawnSpot}
             }
            }]
           properties {waitTime {defaultValue=1}}
          }
         ]
         decorators [
          Is Goal Pawn {
           class=FortBTDecorator_IsGoalPawn
           properties {bInverseCondition=false, flowAbortMode=Both, nodeName=Is Goal Pawn, blackboardKey {bNoneIsAllowedValue=false, selectedKeyName=GoalActor}}
          }
          Has Gameplay Ability {
           class=FortBTDecorator_GameplayAbility_HasGameplayAbility
           properties {gameplayAbilityTagContainer {defaultValue {gameplayTags [{tagName=NPC.Ability.NonAttack.Avoid.Enemy.Wander}]}}}
          }
          Compare Gameplay Ability Distances {
           class=FortBTDecorator_GameplayAbility_CompareDistance
           properties {
            bInverseCondition=false
            flowAbortMode=LowerPriority
            nodeName=Compare Gameplay Ability Distances
            activeAbilityTagsToSkipTestingValue {key=None, defaultValue {gameplayTags []}}
            distanceComparisons [{
              addedOffset=0
              bUseAddedOffset=false
              bUseOverriddenValue=false
              comparisonType=ThreeDimensions
              operator=Less
              overriddenValue=0
              distanceDataTags {gameplayTags [{tagName=NPC.Ability.MaxApproachRange}]}
             }]
            gameplayAbilityTagContainer {key=None, defaultValue {gameplayTags [{tagName=NPC.Ability.NonAttack.Avoid.Enemy.Wander}]}}
            targetValue {baseClass=&Actor, defaultValue=null, key=GoalActor}
           }
          }
          Gameplay Tag Query {
           class=BTDecorator_GameplayTagQuery
           properties {
            flowAbortMode=LowerPriority
            nodeName=Gameplay Tag Query
            gameplayTagQuery {
             autoDescription=" ANY( Gameplay.Status.RecentlyDamaged )"
             queryTokenStream [0,1,1,1,0]
             tagDictionary [{tagName=Gameplay.Status.RecentlyDamaged}]
            }
           }
          }
         ]
         properties {
          children [
           {
            childComposite=null
            childTask=&self:FortBTTask_ExecuteGameplayAbility_10
            decorators [&self:FortBTDecorator_GameplayAbility_CanActivate_4]
           }
           {
            childComposite=null
            childTask=&self:BTTask_Wait_7
            decorators [&self:FortBTDecorator_GameplayAbility_IsOnCooldown_2]
           }
          ]
          services []
         }
         services []
        }
       ]
       decorators [Has NO Attack Ability? {
         class=FortBTDecorator_GameplayAbility_HasGameplayAbility
         properties {
          bInverseCondition=true
          nodeName=Has NO Attack Ability?
          gameplayAbilityTagContainer {defaultValue {gameplayTags [{tagName=NPC.Ability.Attack}]}}
         }
        }]
       properties {
        children [
         {
          childComposite=&self:BTComposite_Selector_43
          childTask=null
          decorators [&self:FortBTDecorator_IsGoalPawn_2]
         }
         {
          childComposite=&self:BTComposite_Selector_28
          childTask=null
          decorators [
           &self:BTDecorator_GameplayTagQuery_10
           &self:FortBTDecorator_GameplayAbility_HasGameplayAbility_23
           &self:BTDecorator_CheckGameplayTagsOnActor_18
          ]
         }
         {
          childComposite=&self:BTComposite_Selector_40
          childTask=null
          decoratorOps [{number=5, operation=And}, {number=0}, {number=1}, {number=2}, {number=2, operation=Or}, {number=3}, {number=4}, {number=5}]
          decorators [
           &self:FortBTDecorator_GameplayAbility_HasGameplayAbility_11
           &self:FortBTDecorator_GameplayAbility_CanActivate_1
           &self:FortBTDecorator_IsGoalPawn_4
           &self:BTDecorator_IsBBEntryOfClass_7
           &self:BTDecorator_IsBBEntryOfClass_8
           &self:BTDecorator_Blackboard_2
          ]
         }
         {
          childComposite=&self:BTComposite_Selector_56
          childTask=null
          decorators [
           &self:FortBTDecorator_GameplayAbility_HasGameplayAbility_18
           &self:BTDecorator_GameplayTagQuery_16
           &self:BTDecorator_Blackboard_8
          ]
         }
         {
          childComposite=&self:BTComposite_Selector_2
          childTask=null
          decoratorOps [{number=3, operation=And}, {number=0}, {number=1}, {number=2, operation=Or}, {number=2}, {number=3}]
          decorators [
           &self:FortBTDecorator_IsGoalPawn_3
           &self:FortBTDecorator_GameplayAbility_HasGameplayAbility_4
           &self:FortBTDecorator_GameplayAbility_CompareDistance_16
           &self:BTDecorator_GameplayTagQuery_24
          ]
         }
        ]
        services []
       }
       services []
      }
     ]
     decorators [
      Blackboard Based Condition {
       class=BTDecorator_Blackboard
       properties {
        arithmeticOperation=Equal
        basicOperation=NotSet
        flowAbortMode=LowerPriority
        intValue=0
        nodeName=Blackboard Based Condition
        notifyObserver=ValueChange
        stringValue=""
        blackboardKey {selectedKeyName=Leader}
       }
      }
      Blackboard Based Condition {
       class=BTDecorator_Blackboard
       properties {
        arithmeticOperation=Equal
        flowAbortMode=LowerPriority
        intValue=0
        nodeName=Blackboard Based Condition
        notifyObserver=ValueChange
        stringValue=""
        blackboardKey {selectedKeyName=Leader}
       }
      }
      AI Close to Leader? {
       class=FortBTDecorator_DistanceBetween
       properties {
        flowAbortMode=LowerPriority
        nodeName=AI Close to Leader?
        operator=LessOrEqual
        blackboardKeyB {selectedKeyName=Leader}
        specifiedDistance {defaultValue=4000}
       }
      }
      Leader close to AI's Goal? {
       class=FortBTDecorator_DistanceBetween
       properties {
        flowAbortMode=LowerPriority
        nodeName=Leader close to AI's Goal?
        operator=LessOrEqual
        blackboardKeyA {selectedKeyName=Leader}
        blackboardKeyB {selectedKeyName=GoalActor}
        specifiedDistance {defaultValue=3000}
       }
      }
     ]
     properties {
      children [
       {
        childComposite=&self:BTComposite_Selector_38
        childTask=null
        decorators [&self:FortBTDecorator_GameplayAbility_HasGameplayAbility_26]
       }
       {
        childComposite=&self:BTComposite_Selector_44
        childTask=null
        decorators [&self:FortBTDecorator_IsGoalPawn_5]
       }
       {
        childComposite=&self:BTComposite_Selector_55
        childTask=null
        decorators [&self:FortBTDecorator_GameplayAbility_HasGameplayAbility_16]
       }
      ]
      services []
     }
     services []
    }
   ]
   decorators [Has GoalActor? {
     class=BTDecorator_Blackboard
     properties {
      arithmeticOperation=Equal
      flowAbortMode=LowerPriority
      intValue=0
      nodeName=Has GoalActor?
      notifyObserver=ValueChange
      stringValue=""
      blackboardKey {selectedKeyName=GoalActor}
     }
    }]
   properties {
    children [
     {
      childComposite=&self:BTComposite_Selector_25
      childTask=null
      decorators [
       &self:BTDecorator_GameplayTagQuery_14
       &self:FortBTDecorator_IsGoalPawn_0
      ]
     }
     {
      childComposite=&self:BTComposite_Selector_10
      childTask=null
      decoratorOps [{number=2, operation=Or}, {number=0}, {number=3, operation=And}, {number=1}, {number=2}, {number=3}]
      decorators [
       &self:BTDecorator_Blackboard_13
       &self:BTDecorator_Blackboard_14
       &self:FortBTDecorator_DistanceBetween_4
       &self:FortBTDecorator_DistanceBetween_5
      ]
     }
    ]
    services [&self:FortBTService_SendGameplayEvents_C_4]
   }
   services [Send Gameplay Events to Blackboard Key - Intro Phase {
     class=FortBTService_SendGameplayEvents_C
     properties {
      activatingPawn=null
      activatingPawnAbilitySystemComponent=null
      bCallTickOnSearchStart=false
      bRestartTimerOnEachActivation=false
      bShowEventDetails=false
      bShowPropertyDetails=true
      customDescription=""
      goalActor=null
      interval=0.5
      nodeName=Send Gameplay Events to Blackboard Key - Intro Phase
      ownerActor=null
      randomDeviation=0.1
      targetAbilitySystemComponent=null
      useEventTickLogic=false
      activationEventKeysTags []
      activationEventValuesBlackboards []
      activationEvents {"(TagName=\"NPC.Behavior.Phase.Intro.Apply\")" {bNoneIsAllowedValue=false, selectedKeyName=SelfActor}}
      deactivationEventKeysTags []
      deactivationEvents {}
      deactivationEventsValuesBlackboards []
     }
    }]
  }
  Selector {
   class=BTComposite_Selector
   children [Sequence {
     class=BTComposite_Sequence
     children [
      Wait {class=BTTask_Wait, properties {waitTime {defaultValue=0.5}}}
      GameplayAbility {
       class=FortBTTask_ExecuteGameplayAbility
       decorators []
       properties {
        nodeName=GameplayAbility
        gameplayAbilityTagContainer {defaultValue {gameplayTags [{tagName=NPC.Ability.NonAttack.Frustration.DueTo.Movement}]}}
       }
      }
     ]
     decorators []
     properties {children [
       {childTask=&self:BTTask_Wait_12}
       {childTask=&self:FortBTTask_ExecuteGameplayAbility_17}
      ]}
    }]
   decorators [
    Is NOT Falling? {
     class=BTDecorator_GameplayTagQuery
     properties {
      flowAbortMode=LowerPriority
      nodeName=Is NOT Falling?
      gameplayTagQuery {autoDescription=" NONE( MovementMode.Falling )", queryTokenStream [0,1,3,1,0], tagDictionary [{tagName=MovementMode.Falling}]}
     }
    }
    Has Gameplay Ability {
     class=FortBTDecorator_GameplayAbility_HasGameplayAbility
     properties {gameplayAbilityTagContainer {defaultValue {gameplayTags [{tagName=NPC.Ability.NonAttack.Frustration.DueTo.Movement}]}}}
    }
    NPC.Status.Frustration.DueTo.Movement {
     class=BTDecorator_CheckGameplayTagsOnActor
     properties {
      bInverseCondition=false
      flowAbortMode=None
      nodeName=NPC.Status.Frustration.DueTo.Movement
      tagsToMatch=Any
      actorToCheck {bNoneIsAllowedValue=false, selectedKeyName=SelfActor}
      gameplayTags {gameplayTags [{tagName=NPC.Status.Frustration.DueTo.Movement}]}
     }
    }
   ]
   properties {children [{childComposite=&self:BTComposite_Sequence_6, childTask=null, decorators []}], services []}
   services []
  }
  Selector {
   class=BTComposite_Selector
   children [
    Injection Point HasLeader.HighPriority {
     class=FortBTTask_InjectionPoint
     children []
     decorators [Can Do Injected Behavior {
       class=FortBTDecorator_CanDoInjectedBehavior
       properties {
        arithmeticOperation=GreaterOrEqual
        bInverseCondition=false
        basicOperation=Set
        floatValue=0
        flowAbortMode=Both
        intValue=3
        nodeName=Can Do Injected Behavior
        notifyObserver=ResultChange
        stringValue=ExecutionPending
        textOperation=Equal
        blackboardKey {bNoneIsAllowedValue=false, selectedKeyName=AIEvaluator_Injection_HasLeader_HighPriority}
       }
      }]
     properties {
      bIgnoreRestartSelf=false
      nodeName=Injection Point HasLeader.HighPriority
      injectedAction {baseClass=&Object, defaultValue=null, key=Injection_HasLeader_HighPriority_ActionObject}
      injectionTag {tagName=Behavior.Injection.HasLeader.HighPriority}
     }
     services []
    }
    GameplayAbility {
     class=FortBTTask_ExecuteGameplayAbility
     decorators [
      Has Gameplay Ability {
       class=FortBTDecorator_GameplayAbility_HasGameplayAbility
       properties {gameplayAbilityTagContainer {defaultValue {gameplayTags [{tagName=NPC.Ability.NonAttack.DropLoot}]}}}
      }
      Can Activate Gameplay Ability {
       class=FortBTDecorator_GameplayAbility_CanActivate
       properties {
        flowAbortMode=LowerPriority
        gameplayAbilityTagContainer {defaultValue {gameplayTags [{tagName=NPC.Ability.NonAttack.DropLoot}]}}
        targetValue {key=GoalActor}
       }
      }
     ]
     properties {nodeName=GameplayAbility, gameplayAbilityTagContainer {defaultValue {gameplayTags [{tagName=NPC.Ability.NonAttack.DropLoot}]}}}
    }
    GameplayAbility {
     class=FortBTTask_ExecuteGameplayAbility
     decorators [
      Has Gameplay Ability {
       class=FortBTDecorator_GameplayAbility_HasGameplayAbility
       properties {gameplayAbilityTagContainer {defaultValue {gameplayTags [{tagName=NPC.Ability.NonAttack.Reposition.AroundLeader}]}}}
      }
      Can Activate Gameplay Ability {
       class=FortBTDecorator_GameplayAbility_CanActivate
       properties {
        flowAbortMode=None
        gameplayAbilityTagContainer {defaultValue {gameplayTags [{tagName=NPC.Ability.NonAttack.Reposition.AroundLeader}]}}
        targetValue {key=SpawnSpot}
       }
      }
      Within MaxApproachRange? {
       class=FortBTDecorator_GameplayAbility_CompareDistance
       properties {
        bInverseCondition=false
        flowAbortMode=Both
        nodeName=Within MaxApproachRange?
        activeAbilityTagsToSkipTestingValue {key=None, defaultValue {gameplayTags []}}
        distanceComparisons [{
          addedOffset=0
          bUseAddedOffset=false
          bUseOverriddenValue=false
          comparisonType=ThreeDimensions
          operator=Less
          overriddenValue=0
          distanceDataTags {gameplayTags [{tagName=NPC.Ability.MaxApproachRange}]}
         }]
        gameplayAbilityTagContainer {key=None, defaultValue {gameplayTags [{tagName=NPC.Ability.NonAttack.Reposition.AroundLeader}]}}
        targetValue {baseClass=&Actor, defaultValue=null, key=Leader}
       }
      }
     ]
     properties {
      nodeName=GameplayAbility
      gameplayAbilityTagContainer {defaultValue {gameplayTags [{tagName=NPC.Ability.NonAttack.Reposition.AroundLeader}]}}
     }
    }
    FortMoveTo {
     class=FortBTTask_GameMoveTo
     decorators [NOT Within MaxApproachRange? {
       class=FortBTDecorator_GameplayAbility_CompareDistance
       properties {
        bInverseCondition=false
        flowAbortMode=Both
        nodeName=NOT Within MaxApproachRange?
        activeAbilityTagsToSkipTestingValue {key=None, defaultValue {gameplayTags []}}
        distanceComparisons [{
          addedOffset=0
          bUseAddedOffset=false
          bUseOverriddenValue=false
          comparisonType=ThreeDimensions
          operator=Greater
          overriddenValue=0
          distanceDataTags {gameplayTags [{tagName=NPC.Ability.MaxApproachRange}]}
         }]
        gameplayAbilityTagContainer {key=None, defaultValue {gameplayTags [{tagName=NPC.Ability.NonAttack.Reposition.AroundLeader}]}}
        targetValue {baseClass=&Actor, defaultValue=null, key=Leader}
       }
      }]
     properties {
      bDeimosFlavor=false
      bFinishMoveOnOverlap=true
      bStopAtGoal=false
      nodeName=FortMoveTo
      acceptableRadius {defaultValue=500}
      bReachTestIncludesAgentRadius {defaultValue=true}
      bReachTestIncludesGoalRadius {defaultValue=true}
      blackboardKey {selectedKeyName=Leader}
      filterClass {defaultValue=&NavFilter_NPC_AnyTraversable_C}
      navFilterTag {tagName=NavFilter.NPC.DefaultMove}
     }
    }
    Wait {
     class=BTTask_Wait
     decorators [Is On Cooldown {
       class=FortBTDecorator_GameplayAbility_IsOnCooldown
       properties {
        bInverseCondition=false
        flowAbortMode=None
        nodeName=Is On Cooldown
        activeAbilityTagsToSkipTestingValue {key=None, defaultValue {gameplayTags []}}
        gameplayAbilityTagContainer {key=None, defaultValue {gameplayTags [{tagName=NPC.Ability.NonAttack.Reposition.AroundLeader}]}}
        targetValue {baseClass=&Actor, defaultValue=null, key=SpawnSpot}
       }
      }]
     properties {waitTime {defaultValue=10000}}
    }
   ]
   decorators [HasLeader? {
     class=BTDecorator_Blackboard
     properties {
      arithmeticOperation=Equal
      intValue=0
      nodeName=HasLeader?
      notifyObserver=ValueChange
      stringValue=""
      blackboardKey {selectedKeyName=Leader}
     }
    }]
   properties {
    children [
     {
      childComposite=null
      childTask=&self:FortBTTask_InjectionPoint_0
      decorators [&self:FortBTDecorator_CanDoInjectedBehavior_0]
     }
     {
      childComposite=null
      childTask=&self:FortBTTask_ExecuteGameplayAbility_28
      decorators [
       &self:FortBTDecorator_GameplayAbility_HasGameplayAbility_31
       &self:FortBTDecorator_GameplayAbility_CanActivate_13
      ]
     }
     {
      childComposite=null
      childTask=&self:FortBTTask_ExecuteGameplayAbility_31
      decorators [
       &self:FortBTDecorator_GameplayAbility_HasGameplayAbility_35
       &self:FortBTDecorator_GameplayAbility_CanActivate_22
       &self:FortBTDecorator_GameplayAbility_CompareDistance_1
      ]
     }
     {
      childComposite=null
      childTask=&self:FortBTTask_GameMoveTo_0
      decorators [&self:FortBTDecorator_GameplayAbility_CompareDistance_2]
     }
     {
      childComposite=null
      childTask=&self:BTTask_Wait_3
      decorators [&self:FortBTDecorator_GameplayAbility_IsOnCooldown_3]
     }
    ]
    services [&self:FortBTService_AIEvaluatorsInjector_0]
   }
   services [Behavior Injector {
     class=FortBTService_AIEvaluatorsInjector
     properties {
      bCallTickOnSearchStart=false
      bRestartTimerOnEachActivation=false
      interval=0.5
      nodeName=Behavior Injector
      randomDeviation=0.1
      injectionTagsKeys [{
        actionInterfaceObjectKeyName=Injection_HasLeader_HighPriority_ActionObject
        injectionKeyName=AIEvaluator_Injection_HasLeader_HighPriority
        injectionTag {tagName=Behavior.Injection.HasLeader.HighPriority}
       }]
     }
    }]
  }
  Selector {
   class=BTComposite_Selector
   children [
    GameplayAbility {
     class=FortBTTask_ExecuteGameplayAbility
     decorators [Can Activate Gameplay Ability {
       class=FortBTDecorator_GameplayAbility_CanActivate
       properties {
        flowAbortMode=None
        gameplayAbilityTagContainer {defaultValue {gameplayTags [{tagName=NPC.Ability.NonAttack.Move.ToLand}]}}
        targetValue {key=SpawnSpot}
       }
      }]
     properties {nodeName=GameplayAbility, gameplayAbilityTagContainer {defaultValue {gameplayTags [{tagName=NPC.Ability.NonAttack.Move.ToLand}]}}}
    }
    Wait {
     class=BTTask_Wait
     decorators [Is On Cooldown {
       class=FortBTDecorator_GameplayAbility_IsOnCooldown
       properties {
        bInverseCondition=false
        flowAbortMode=None
        nodeName=Is On Cooldown
        activeAbilityTagsToSkipTestingValue {key=None, defaultValue {gameplayTags []}}
        gameplayAbilityTagContainer {key=None, defaultValue {gameplayTags [{tagName=NPC.Ability.NonAttack.Move.ToLand}]}}
        targetValue {baseClass=&Actor, defaultValue=null, key=SpawnSpot}
       }
      }]
     properties {waitTime {defaultValue=1}}
    }
   ]
   decorators [
    Is Surface Swimming {
     class=BTDecorator_GameplayTagQuery
     properties {
      flowAbortMode=LowerPriority
      nodeName=Is Surface Swimming
      gameplayTagQuery {autoDescription=" ANY( MovementMode.SurfaceSwimming )", queryTokenStream [0,1,1,1,0], tagDictionary [{tagName=MovementMode.SurfaceSwimming}]}
     }
    }
    Has Gameplay Ability {
     class=FortBTDecorator_GameplayAbility_HasGameplayAbility
     properties {gameplayAbilityTagContainer {defaultValue {gameplayTags [{tagName=NPC.Ability.NonAttack.Move.ToLand}]}}}
    }
   ]
   properties {
    children [
     {
      childComposite=null
      childTask=&self:FortBTTask_ExecuteGameplayAbility_19
      decorators [&self:FortBTDecorator_GameplayAbility_CanActivate_21]
     }
     {
      childComposite=null
      childTask=&self:BTTask_Wait_14
      decorators [&self:FortBTDecorator_GameplayAbility_IsOnCooldown_9]
     }
    ]
    services []
   }
   services []
  }
  BTComposite_Selector_18 {
   class=BTComposite_Selector
   children [
    GameplayAbility {
     class=FortBTTask_ExecuteGameplayAbility
     decorators [Can Activate Gameplay Ability {
       class=FortBTDecorator_GameplayAbility_CanActivate
       properties {
        flowAbortMode=None
        gameplayAbilityTagContainer {defaultValue {gameplayTags [{tagName=NPC.Ability.NonAttack.Move.ToPatrol}]}}
        targetValue {key=SpawnSpot}
       }
      }]
     properties {nodeName=GameplayAbility, gameplayAbilityTagContainer {defaultValue {gameplayTags [{tagName=NPC.Ability.NonAttack.Move.ToPatrol}]}}}
    }
    Wait {
     class=BTTask_Wait
     decorators [Is On Cooldown {
       class=FortBTDecorator_GameplayAbility_IsOnCooldown
       properties {
        bInverseCondition=false
        flowAbortMode=None
        nodeName=Is On Cooldown
        activeAbilityTagsToSkipTestingValue {key=None, defaultValue {gameplayTags []}}
        gameplayAbilityTagContainer {key=None, defaultValue {gameplayTags [{tagName=NPC.Ability.NonAttack.Move.ToPatrol}]}}
        targetValue {baseClass=&Actor, defaultValue=null, key=SpawnSpot}
       }
      }]
     properties {waitTime {defaultValue=1}}
    }
   ]
   decorators [Has Gameplay Ability {
     class=FortBTDecorator_GameplayAbility_HasGameplayAbility
     properties {gameplayAbilityTagContainer {defaultValue {gameplayTags [{tagName=NPC.Ability.NonAttack.Move.ToPatrol}]}}}
    }]
   properties {
    nodeName=""
    children [
     {
      childComposite=null
      childTask=&self:FortBTTask_ExecuteGameplayAbility_6
      decorators [&self:FortBTDecorator_GameplayAbility_CanActivate_28]
     }
     {
      childComposite=null
      childTask=&self:BTTask_Wait_20
      decorators [&self:FortBTDecorator_GameplayAbility_IsOnCooldown_6]
     }
    ]
    services []
   }
   services []
  }
  Selector {
   class=BTComposite_Selector
   children [
    GameplayAbility {
     class=FortBTTask_ExecuteGameplayAbility
     decorators [
      Has Gameplay Ability {
       class=FortBTDecorator_GameplayAbility_HasGameplayAbility
       properties {gameplayAbilityTagContainer {defaultValue {gameplayTags [{tagName=NPC.Ability.NonAttack.PlayMontage.Wander.Freely}]}}}
      }
      Has Phase Wander Animation Pending {
       class=BTDecorator_GameplayTagQuery
       properties {
        flowAbortMode=None
        nodeName=Has Phase Wander Animation Pending
        gameplayTagQuery {
         autoDescription=" ANY( NPC.Behavior.Phase.Wander.Freely.Animation.Pending )"
         queryTokenStream [0,1,1,1,0]
         tagDictionary [{tagName=NPC.Behavior.Phase.Wander.Freely.Animation.Pending}]
        }
       }
      }
      Can Activate Gameplay Ability {
       class=FortBTDecorator_GameplayAbility_CanActivate
       properties {
        flowAbortMode=None
        gameplayAbilityTagContainer {defaultValue {gameplayTags [{tagName=NPC.Ability.NonAttack.PlayMontage.Wander.Freely}]}}
        targetValue {key=SpawnSpot}
       }
      }
     ]
     properties {
      nodeName=GameplayAbility
      gameplayAbilityTagContainer {defaultValue {gameplayTags [{tagName=NPC.Ability.NonAttack.PlayMontage.Wander.Freely}]}}
     }
    }
    Selector {
     class=BTComposite_Selector
     children [Wait {class=BTTask_Wait, properties {waitTime {defaultValue=1000}}}]
     decorators [Is On Cooldown {
       class=FortBTDecorator_GameplayAbility_IsOnCooldown
       properties {
        bInverseCondition=false
        flowAbortMode=Both
        nodeName=Is On Cooldown
        activeAbilityTagsToSkipTestingValue {key=None, defaultValue {gameplayTags []}}
        gameplayAbilityTagContainer {key=None, defaultValue {gameplayTags [{tagName=NPC.Ability.NonAttack.Wander.Freely}]}}
        targetValue {baseClass=&Actor, defaultValue=null, key=SpawnSpot}
       }
      }]
     properties {children [{childComposite=null, childTask=&self:BTTask_Wait_15, decorators []}], services []}
     services []
    }
    GameplayAbility {
     class=FortBTTask_ExecuteGameplayAbility
     decorators [
      Can Activate Gameplay Ability {
       class=FortBTDecorator_GameplayAbility_CanActivate
       properties {
        flowAbortMode=None
        gameplayAbilityTagContainer {defaultValue {gameplayTags [{tagName=NPC.Ability.NonAttack.Wander.Freely}]}}
        targetValue {key=SpawnSpot}
       }
      }
      Has NO GoalActor? {
       class=BTDecorator_Blackboard
       properties {
        arithmeticOperation=Equal
        basicOperation=NotSet
        intValue=0
        nodeName=Has NO GoalActor?
        stringValue=""
        blackboardKey {selectedKeyName=GoalActor}
       }
      }
     ]
     properties {nodeName=GameplayAbility, gameplayAbilityTagContainer {defaultValue {gameplayTags [{tagName=NPC.Ability.NonAttack.Wander.Freely}]}}}
    }
    GameplayAbility {
     class=FortBTTask_ExecuteGameplayAbility
     decorators [
      Can Activate Gameplay Ability {
       class=FortBTDecorator_GameplayAbility_CanActivate
       properties {
        flowAbortMode=None
        gameplayAbilityTagContainer {defaultValue {gameplayTags [{tagName=NPC.Ability.NonAttack.Wander.Freely}]}}
        targetValue {key=SpawnSpot}
       }
      }
      Has GoalActor? {
       class=BTDecorator_Blackboard
       properties {
        arithmeticOperation=Equal
        intValue=0
        nodeName=Has GoalActor?
        notifyObserver=ValueChange
        stringValue=""
        blackboardKey {selectedKeyName=GoalActor}
       }
      }
     ]
     properties {nodeName=GameplayAbility, gameplayAbilityTagContainer {defaultValue {gameplayTags [{tagName=NPC.Ability.NonAttack.Wander.Freely}]}}}
    }
   ]
   decorators [Has Gameplay Ability {
     class=FortBTDecorator_GameplayAbility_HasGameplayAbility
     properties {gameplayAbilityTagContainer {defaultValue {gameplayTags [{tagName=NPC.Ability.NonAttack.Wander.Freely}]}}}
    }]
   properties {
    children [
     {
      childComposite=null
      childTask=&self:FortBTTask_ExecuteGameplayAbility_20
      decorators [
       &self:FortBTDecorator_GameplayAbility_HasGameplayAbility_25
       &self:BTDecorator_GameplayTagQuery_17
       &self:FortBTDecorator_GameplayAbility_CanActivate_16
      ]
     }
     {
      childComposite=&self:BTComposite_Selector_36
      childTask=null
      decorators [&self:FortBTDecorator_GameplayAbility_IsOnCooldown_4]
     }
     {
      childComposite=null
      childTask=&self:FortBTTask_ExecuteGameplayAbility_3
      decorators [
       &self:FortBTDecorator_GameplayAbility_CanActivate_17
       &self:BTDecorator_Blackboard_3
      ]
     }
     {
      childComposite=null
      childTask=&self:FortBTTask_ExecuteGameplayAbility_13
      decorators [
       &self:FortBTDecorator_GameplayAbility_CanActivate_9
       &self:BTDecorator_Blackboard_7
      ]
     }
    ]
    services []
   }
   services []
  }
  Wait {class=BTTask_Wait, properties {waitTime {defaultValue=1000}}}
 ]
 decorators []
 properties {
  children [
   {
    childComposite=&self:BTComposite_Selector_52
    childTask=null
    decorators [&self:BTDecorator_GameplayTagQuery_4]
   }
   {
    childComposite=null
    childTask=&self:BTTask_Wait_17
    decorators [&self:BTDecorator_GameplayTagQuery_13]
   }
   {
    childComposite=null
    childTask=&self:BTTask_RunBehavior_1
    decorators [&self:BTDecorator_Blackboard_10]
   }
   {
    childComposite=&self:BTComposite_Selector_42
    childTask=null
    decorators [&self:BTDecorator_GameplayTagQuery_20]
   }
   {
    childComposite=&self:BTComposite_Selector_49
    childTask=null
    decorators [
     &self:BTDecorator_GameplayTagQuery_1
     &self:FortBTDecorator_GameplayAbility_HasGameplayAbility_32
    ]
   }
   {
    childComposite=null
    childTask=&self:FortBTTask_ExecuteGameplayAbility_29
    decorators [
     &self:BTDecorator_GameplayTagQuery_7
     &self:FortBTDecorator_GameplayAbility_HasGameplayAbility_37
    ]
   }
   {
    childComposite=&self:BTComposite_Selector_50
    childTask=null
    decorators [&self:BTDecorator_GameplayTagQuery_3]
   }
   {
    childComposite=&self:BTComposite_Selector_9
    childTask=null
    decorators [&self:BTDecorator_Blackboard_1]
   }
   {
    childComposite=&self:BTComposite_Selector_29
    childTask=null
    decorators [
     &self:BTDecorator_GameplayTagQuery_11
     &self:FortBTDecorator_GameplayAbility_HasGameplayAbility_19
     &self:BTDecorator_CheckGameplayTagsOnActor_3
    ]
   }
   {
    childComposite=&self:BTComposite_Selector_53
    childTask=null
    decorators [&self:BTDecorator_Blackboard_9]
   }
   {
    childComposite=&self:BTComposite_Selector_34
    childTask=null
    decorators [
     &self:BTDecorator_GameplayTagQuery_8
     &self:FortBTDecorator_GameplayAbility_HasGameplayAbility_30
    ]
   }
   {
    childComposite=&self:BTComposite_Selector_18
    childTask=null
    decorators [&self:FortBTDecorator_GameplayAbility_HasGameplayAbility_10]
   }
   {
    childComposite=&self:BTComposite_Selector_4
    childTask=null
    decorators [&self:FortBTDecorator_GameplayAbility_HasGameplayAbility_5]
   }
   {childComposite=null, childTask=&self:BTTask_Wait_1, decorators []}
  ]
  services [&self:FortAthenaBTContext_SuppressAutomaticAttackCheck_1]
 }
 services [Athena Suppress Automatic Attack Check {
   class=FortAthenaBTContext_SuppressAutomaticAttackCheck
   properties {
    bCallTickOnSearchStart=false
    bRestartTimerOnEachActivation=false
    interval=0.5
    nodeName=Athena Suppress Automatic Attack Check
    randomDeviation=0.1
    enabled {defaultValue=true, key=None}
   }
  }]
}
