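# Behavior tree for Vk creature AI, using BB_Vk_Creature, that holds in long Wait tasks rather than running actions itself.
# A looping selector, named to prevent evaluator restarts, checks NavigateTo, PlayAnimation, LookAt and Idle execution-status keys in priority order, waiting on the first at ExecutionPending or higher.
# An unconditional Idle wait is the fallback.
# A root service suppresses goal updates and unregisters the creature from the goal manager.
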BehaviorTree /Game/Athena/AI/BehaviorTrees/BT_Vk_Creature_AI
blackboard=/Game/Athena/AI/Blackboards/BB_Vk_Creature
root {
 name=Selector
 class=BTComposite_Selector
 children [Loop to prevent evaluator to restart {
   class=BTComposite_Selector
   children [
    Wait {
     class=BTTask_Wait
     decorators [NavigateTo Status {
       class=BTDecorator_Blackboard
       properties {
        arithmeticOperation=GreaterOrEqual
        intValue=3
        nodeName=NavigateTo Status
        stringValue=EExecutionStatus::ExecutionPending
        blackboardKey {selectedKeyName=VK_NavigateToExecutionStatus}
       }
      }]
     properties {waitTime {defaultValue=10000000000}}
    }
    Wait {
     class=BTTask_Wait
     decorators [PlayAnimation Status {
       class=BTDecorator_Blackboard
       properties {
        arithmeticOperation=GreaterOrEqual
        intValue=3
        nodeName=PlayAnimation Status
        stringValue=ExecutionPending
        blackboardKey {selectedKeyName=Vk_PlayAnimationExecutionStatus}
       }
      }]
     properties {waitTime {defaultValue=10000000000}}
    }
    Wait {
     class=BTTask_Wait
     decorators [Look At Status {
       class=BTDecorator_Blackboard
       properties {
        arithmeticOperation=GreaterOrEqual
        intValue=3
        nodeName=Look At Status
        notifyObserver=ValueChange
        stringValue=ExecutionPending
        blackboardKey {selectedKeyName=Vk_LookAtExecutionStatus}
       }
      }]
     properties {waitTime {defaultValue=10000000000}}
    }
    Wait {
     class=BTTask_Wait
     decorators [Idle Status {
       class=BTDecorator_Blackboard
       properties {
        arithmeticOperation=GreaterOrEqual
        intValue=3
        nodeName=Idle Status
        notifyObserver=ValueChange
        stringValue=ExecutionPending
        blackboardKey {selectedKeyName=Vk_IdleExecutionStatus}
       }
      }]
     properties {waitTime {defaultValue=10000000000}}
    }
    Idle {class=BTTask_Wait, properties {nodeName=Idle, waitTime {defaultValue=10000000000}}}
   ]
   decorators [Loop {class=BTDecorator_Loop}]
   properties {
    nodeName=Loop to prevent evaluator to restart
    children [
     {
      childComposite=null
      childTask=&self:BTTask_Wait_4
      decorators [&self:BTDecorator_Blackboard_5]
     }
     {
      childComposite=null
      childTask=&self:BTTask_Wait_3
      decorators [&self:BTDecorator_Blackboard_4]
     }
     {
      childComposite=null
      childTask=&self:BTTask_Wait_1
      decorators [&self:BTDecorator_Blackboard_2]
     }
     {
      childComposite=null
      childTask=&self:BTTask_Wait_2
      decorators [&self:BTDecorator_Blackboard_3]
     }
     {childComposite=null, childTask=&self:BTTask_Wait_0, decorators []}
    ]
    services []
   }
   services []
  }]
 decorators []
 properties {
  children [{
    childComposite=&self:BTComposite_Selector_3
    childTask=null
    decorators [&self:BTDecorator_Loop_4]
   }]
  services [&self:FortBTContext_SuppressGoalUpdate_0]
 }
 services [Suppress Goal Updates {
   class=FortBTContext_SuppressGoalUpdate
   properties {
    bCallTickOnSearchStart=false
    bRestartTimerOnEachActivation=false
    bUnregisterFromGoalManager=true
    interval=0.5
    nodeName=Suppress Goal Updates
    randomDeviation=0.1
    enabled {defaultValue=true, key=None}
   }
  }]
}
