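# Top-level behavior tree for Vk (Verse) terrestrial creature AI, choosing between default and Verse-driven behavior on the BB_Vk_Creature blackboard.
# A root Selector first runs the BT_Terrestrial subtree, gated by a blackboard decorator requiring VK_DefaultBehaviorExecutionStatus to be at least ExecutionPending.
# If that branch is blocked or fails, the Selector falls back to the ungated BT_Vk_Creature_AI Verse subtree.
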BehaviorTree /Irwin/Athena/AI/BehaviorTrees/BT_Vk_Terrestrial
blackboard=/Game/Athena/AI/Blackboards/BB_Vk_Creature
root {
 name=Selector
 class=BTComposite_Selector
 children [
  Run Default Behavior {
   class=BTTask_RunBehavior
   decorators [Default Behavior Status {
     class=BTDecorator_Blackboard
     properties {
      arithmeticOperation=GreaterOrEqual
      intValue=3
      nodeName=Default Behavior Status
      stringValue=EExecutionStatus::ExecutionPending
      blackboardKey {selectedKeyName=VK_DefaultBehaviorExecutionStatus}
     }
    }]
   properties {behaviorAsset=/Game/Athena/AI/BehaviorTrees/BT_Terrestrial, nodeName=Run Default Behavior}
  }
  Run Verse Behavior {
   class=BTTask_RunBehavior
   decorators []
   properties {behaviorAsset=/Game/Athena/AI/BehaviorTrees/BT_Vk_Creature_AI, nodeName=Run Verse Behavior}
  }
 ]
 decorators []
 properties {
  children [
   {
    childComposite=null
    childTask=&self:BTTask_RunBehavior_0
    decorators [&self:BTDecorator_Blackboard_47]
   }
   {childComposite=null, childTask=&self:BTTask_RunBehavior_1, decorators []}
  ]
  services []
 }
 services []
}
